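# Material function for the Asteria landscape that blends cliff material over an incoming layer material based on slope and height.
# An inverted, boosted slope mask, overridable by input, drives height lerps between the input and cliff heightmaps to form the cliff masks.
# MF_AsteriaCommon_Cliff supplies the cliff, with a static switch for an alternate material.
# It also outputs blended displacement, snow and grass heightmaps, a grass-spawner mask, and a second MF_BlendLandscapeLayers blend.
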MaterialFunction /Game/Athena/Asteria/Environments/Landscape/Materials/MaterialFunctions/MF_AsteriaCliffBlend
graph {nodes [
  FunctionInput(Preview=Constant()) {bUsePreviewValueAsDefault=true, inputName=AsphaltLayer, inputType=FunctionInput_Scalar, sortPriority=4}
  n1 = FunctionInput() {inputName=Material, inputType=FunctionInput_MaterialAttributes, sortPriority=-2}
  n2 = NamedRerouteUsage()
  FunctionOutput(None=BlendMaterialAttributes(A=NamedRerouteUsage(), B=NamedRerouteDeclaration(None=@n1) {name=InputMaterial, nodeColor=<rgba 0.5703 1 0 1>}, Alpha=@n2)) {outputName=Material, sortPriority=-5}
  n3 = NamedRerouteDeclaration(None=FunctionInput(Preview=Constant() {r=1}) {bUsePreviewValueAsDefault=true, inputName=Heightmap, inputType=FunctionInput_Scalar, sortPriority=-1}) {name=InputMaterialHeight, nodeColor=<rgba 0.2187 1 0 1>}
  n4 = Add(A=@n3, B=Constant() {r=-1})
  FunctionOutput(None=LinearInterpolate(A=NamedRerouteUsage(), B=Saturate(None=@n4), Alpha=NamedRerouteUsage())) {outputName=HeightmapForSnow, sortPriority=-3}
  n5 = NamedRerouteUsage()
  n6 = ScalarParameter() {parameterName=SnowScatterInfluence}
  n7 = MaterialFunctionCall!(
   Material 1=NamedRerouteUsage()
   Height Texture 1=NamedRerouteUsage()
   Material 2=NamedRerouteUsage()
   Height Texture 2=NamedRerouteUsage()
   Transistion Phase=Multiply(A=NamedRerouteUsage(), B=Clamp(None=@n5, Min=@n6))
   UseContrast=StaticBool() {value=true}
   Contrast=ScalarParameter() {defaultValue=0.5, group=Cliff, parameterName=CliffHeightBlendContrast}
   AddLayerMask
  ) {materialFunction=/Game/Athena/Apollo/Environments/Landscape/Materials/Biome/CommonMaterials/MF_BlendLandscapeLayers}
  FunctionOutput(None=@n7.BlendedMaterial) {outputName=OutputMaterial2}
  FunctionOutput(None=@n7.Alpha) {outputName=OutPut CliffMask2}
  FunctionOutput(None=@n7.Resulting Height Map) {outputName=RegularHeightmap2}
  FunctionOutput(None=LinearInterpolate(A=NamedRerouteUsage(), Alpha=@n2) {constB=0}) {outputName=CliffEmissiveMask}
  n8 = NamedRerouteUsage()
  n9 = Saturate(None=Multiply(A=OneMinus(None=@n8), B=Constant() {r=1.4}))
  n10 = NamedRerouteDeclaration(None=FunctionOutput(None=FunctionInput(Preview=@n9) {bUsePreviewValueAsDefault=true, inputName=NewSlopeMaskMask, inputType=FunctionInput_Scalar}) {outputName=SloopeMaskNew}) {name=NewSlopeMaskMask, nodeColor=<rgba 0.3359 0 1 1>}
  n11 = MaterialFunctionCall!(
   Diffuse 1
   Diffuse 2
   Transistion Phase=@n10
   Height Texture 1=@n3
   Height Texture 2=NamedRerouteUsage()
   Contrast=ScalarParameter() {defaultValue=0.5, group=Cliff, parameterName=CliffHeightBlendContrast}
  ) {materialFunction=/Engine/Functions/Engine_MaterialFunctions02/Texturing/HeightLerpWithTwoHeightMaps}
  n12 = Add(A=@n11.Lerp Alpha No Contrast, B=Constant() {r=-0.42})
  n13 = Saturate(None=Add(A=Saturate(None=@n12), B=NamedRerouteUsage()))
  FunctionOutput(None=Ceil(None=OneMinus(None=@n13))) {
   desc=Mask for the landscape grass type spawner
   description=Mask for the landscape grass type spawner
   outputName=LandscapeGrassMeshMask
   sortPriority=3
  }
  n14 = ComponentMask(None=VertexNormalWS()) {b=true}
  FunctionOutput(None=NamedRerouteDeclaration(None=Abs(None=@n14)) {name=SlopeMask, nodeColor=<rgba 0.4766 0 1 1>}) {outputName=SlopeMask, sortPriority=7}
  FunctionOutput(None=MaterialFunctionCall(Transition Phase=@n10, Height Texture=NamedRerouteUsage()) {materialFunction=&HeightLerp}.Lerp Alpha No Contrast) {
   desc="This is the heightmap for the grass to make a smooth transition\r\nto the cliff w/o showing too much grass on the rocks"
   description=This is the heightmap for the grass to make a smooth transition to the cliff
   outputName=HeightmapForGrass
   sortPriority=2
  }
  FunctionOutput(None=@n11.Resulting Height Map) {outputName=RegularHeightmap, sortPriority=-4}
  n15 = FunctionInput() {
   bUsePreviewValueAsDefault=true
   desc="This input should come from the\r\nMF_CommonLandscapeLayers\r\ndisplacement output"
   inputName=Displacement
   inputType=FunctionInput_Scalar
   sortPriority=1
  }
  n16 = NamedRerouteUsage()
  n17 = Saturate(None=Multiply(A=Constant() {r=1.75}, B=OneMinus(None=@n16)))
  n18 = NamedRerouteUsage()
  n19 = Reroute() {desc=Displacement from MF_Common_Cliff}
  FunctionOutput(None=LinearInterpolate(
    A=@n15
    B=@n19
    Alpha=MaterialFunctionCall(Transition Phase=@n17, Height Texture=@n18) {materialFunction=&HeightLerp}.Lerp Alpha No Contrast
   )) {outputName=Displacement, sortPriority=5}
  FunctionOutput(None=Constant()) {outputName=SlopeScreeMask, sortPriority=8}
  FunctionOutput(None=@n11.Lerp Alpha No Contrast) {outputName=CliffMaskNoContrast, sortPriority=9}
  LinearInterpolate(A=@n19, B=@n15, Alpha=NamedRerouteUsage())
  n20 = Constant() {r=2}
  n21 = NamedRerouteUsage()
  n22 = Constant() {r=10}
  n23 = OneMinus(None=@n11.Lerp Alpha No Contrast)
  Multiply(A=Saturate(None=Multiply(A=@n20, B=@n21) {constA=2}), B=Saturate(None=Multiply(A=@n22, B=@n23)))
  n24 = Constant()
  n25 = FunctionInput() {bUsePreviewValueAsDefault=true, inputName=CustomBiomeMask, inputType=FunctionInput_Scalar, previewValue=<1 0 0 1>}
  n26 = MaterialFunctionCall!(CustomBiomeMask=@n25, Slope=NamedRerouteUsage()) {materialFunction=&/MF_AsteriaCommon_Cliff}
  n27 = FunctionInput(Preview=StaticBool()) {bUsePreviewValueAsDefault=true, inputName=UseAlternateCliffMat, inputType=FunctionInput_StaticBool, sortPriority=5}
  NamedRerouteDeclaration(None=StaticSwitch(
    True=FunctionInput(Preview=@n24) {bUsePreviewValueAsDefault=true, inputName=AltCliffHeight, inputType=FunctionInput_Scalar, sortPriority=6}
    False=@n26.Height
    Value=@n27
   )) {name=Height, nodeColor=<rgba 0.4766 0 1 1>}
  NamedRerouteDeclaration(None=FunctionOutput(None=@n11.Alpha) {outputName=CliffMask, sortPriority=4}) {name=CliffMask, nodeColor=<rgba 0.4766 0 1 1>}
  n28 = SetMaterialAttributes!(MaterialAttributes)
  NamedRerouteDeclaration(None=StaticSwitch(
    True=FunctionInput(Preview=@n28) {bUsePreviewValueAsDefault=true, inputName=AltCliffMat, inputType=FunctionInput_MaterialAttributes, sortPriority=7}
    False=@n26.Material
    Value=@n27
   )) {name=CliffMaterial, nodeColor=<rgba 0.4766 0 1 1>}
  n29 = Constant()
  NamedRerouteDeclaration(None=StaticSwitch(
    True=FunctionInput(Preview=@n29) {bUsePreviewValueAsDefault=true, inputName=AltCliffSlopeScreeMask, inputType=FunctionInput_Scalar, sortPriority=4}
    False=@n26.SlopeScreeMask
    Value=@n27
   )) {name=SlopeScreeMaskIn, nodeColor=<rgba 0.4766 0 1 1>}
  NamedRerouteDeclaration(None=@n26.CliffEmissiveMask) {name=ClifEmissiveMask, nodeColor=<rgba 0.125 0 1 1>}
  NamedRerouteDeclaration(None=FunctionInput(Preview=Constant() {r=1}) {bUsePreviewValueAsDefault=true, inputName=CliffSnowMask, inputType=FunctionInput_Scalar}) {name=SC_SnowMask, nodeColor=<rgba 1 0.2812 0 1>}
 ]}
