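# Material function for the Asteria landscape that blends common paint layers over a grass, farm-patch or custom base.
# Height-blends gravel, mud, sand or jungle floor, and asphalt or gameplay mud, with static switches choosing variants and sand order.
# Noise breaks up the asphalt-layer edge.
# Outputs material, heightmap, grass and gravel data, and foliage masks; a zeroed layer term keeps landscape-tool target layers working.
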MaterialFunction /Game/Athena/Asteria/Environments/Landscape/Materials/MaterialFunctions/MF_AsteriaCommonLandscape
graph {nodes [
  Constant() {r=1}
  Constant() {r=20}
  Constant() {r=-1}
  n1 = ComponentMask(None=WorldPosition() {worldPositionShaderOffset=WPT_ExcludeAllShaderOffsets}.XYZ) {g=true, r=true}
  Divide(A=@n1, B=ScalarParameter() {defaultValue=1, parameterName=GrassDirtNoiseScale})
  n2 = StaticSwitchParameter(True=NamedRerouteUsage(), False=NamedRerouteUsage()) {parameterName=UseFarmPatches}
  n3 = MaterialFunctionCall() {materialFunction=&/MF_JungleFloor}
  n4 = MaterialFunctionCall!(BiomeMask-RGBA=NamedRerouteUsage(), BiomeColors-UseDefault, UseUnderOceanSandMaterial) {materialFunction=&/MF_Asteria_Sand}
  n5 = LandscapeLayerSample() {parameterName=Sand}
  n6 = MaterialFunctionCall!(
   Material 1=@n2
   Height Texture 1
   Material 2=StaticSwitchParameter(True=@n3.Material, False=@n4.Material) {parameterName=UseJungleFloor}
   Height Texture 2=StaticSwitchParameter(True=@n3.Height, False=@n4.Height) {parameterName=UseJungleFloor}
   Transistion Phase=@n5
   UseContrast
   Contrast
   AddLayerMask
  ) {materialFunction=/Game/Athena/Apollo/Environments/Landscape/Materials/Biome/CommonMaterials/MF_BlendLandscapeLayers}
  n7 = NamedRerouteUsage()
  n8 = NamedRerouteUsage()
  n9 = NamedRerouteUsage()
  n10 = Max(A=LandscapeLayerSample() {parameterName=Gravel}, B=NamedRerouteUsage())
  n11 = MaterialFunctionCall!(LayerWeight=@n10, VarianceMask=NamedRerouteUsage(), BiomeMask-RGBA=NamedRerouteUsage(), BiomeColors-UseDefault) {materialFunction=&/MF_Asteria_Gravel}
  n12 = MaterialFunctionCall(
   Material 1=StaticSwitch(True=@n6.BlendedMaterial, False=@n2, Value=@n7)
   Height Texture 1=StaticSwitch(True=@n6.Resulting Height Map, False=StaticSwitchParameter(True=@n8, False=@n9) {parameterName=UseFarmPatches}, Value=@n7)
   Material 2=@n11.Material
   Height Texture 2=@n11.Material
   Transistion Phase=@n10
  ) {materialFunction=&MF_BlendLandscapeLayers}
  n13 = MaterialFunctionCall!(
   Spec=NamedRerouteUsage()
   Diffuse=NamedRerouteUsage()
   Height=NamedRerouteUsage()
   BiomeColors-UseDefault
   Rough=NamedRerouteUsage()
   Normals=NamedRerouteUsage()
   VarianceMask=NamedRerouteUsage()
   BiomeMask-RGBA=NamedRerouteUsage()
   BiomeRoughMod
  ) {materialFunction=&/MF_Asteria_Mud}
  n14 = MaterialFunctionCall(
   Material 1=@n12.BlendedMaterial
   Height Texture 1=@n12.BlendedMaterial
   Material 2=@n13.Material
   Height Texture 2=@n13.Material
   Transistion Phase=LandscapeLayerSample() {parameterName=Mud}
  ) {materialFunction=&MF_BlendLandscapeLayers}
  n15 = MaterialFunctionCall() {materialFunction=&MF_JungleFloor}
  n16 = MaterialFunctionCall(BiomeMask-RGBA=NamedRerouteUsage()) {materialFunction=&MF_Asteria_Sand}
  n17 = MaterialFunctionCall(
   Material 1=@n14.BlendedMaterial
   Height Texture 1=@n14.BlendedMaterial
   Material 2=StaticSwitchParameter(True=@n15.Material, False=@n16.Material) {parameterName=UseJungleFloor}
   Height Texture 2=StaticSwitchParameter(True=@n15.Height, False=@n16.Height) {parameterName=UseJungleFloor}
   Transistion Phase=LandscapeLayerSample() {parameterName=Sand}
  ) {materialFunction=&MF_BlendLandscapeLayers}
  n18 = NamedRerouteUsage()
  n19 = MaterialFunctionCall!(FallBackMobileMudGrayscale) {materialFunction=&/MF_JungleGameplayMud}
  n20 = MaterialFunctionCall!(TilingSize) {materialFunction=&/MF_Asteria_Asphault}
  n21 = LandscapeLayerCoords()
  n22 = ScalarParameter() {defaultValue=3, parameterName=GameplayMudEdgeBreakupSacle}
  n23 = Subtract(A=TextureSampleParameter2D(UVs=Divide(A=@n21, B=@n22)) {parameterName=GamePlayMudBreakup, texture=/Game/Effects/FogVolume/noise_flamy_Tile}.R) {constB=0.5}
  n24 = ScalarParameter() {defaultValue=0.2, parameterName=GameplayMudEdgeBreakupAmount}
  n25 = MaterialFunctionCall(
   Material 1=StaticSwitch(True=@n14.BlendedMaterial, False=@n17.BlendedMaterial, Value=@n18)
   Height Texture 1=StaticSwitch(True=@n14.Resulting Height Map, False=@n17.Resulting Height Map, Value=@n18)
   Material 2=StaticSwitchParameter(True=@n19.Material, False=@n20.DefaultLandscapeMaterial) {parameterName=UseJungleMud}
   Height Texture 2=StaticSwitchParameter(True=@n19.Height, False=@n20.CrackHeight) {parameterName=UseJungleMud}
   Transistion Phase=Add(A=LandscapeLayerSample() {parameterName=Asphalt}, B=Multiply(A=@n23, B=@n24) {constB=0.2})
   UseContrast=StaticBoolParameter() {defaultValue=true, parameterName=UseGamePlayMudConstrast}
   Contrast=ScalarParameter() {defaultValue=1.5, parameterName=GamePlayMudLayerConstrast}
  ) {materialFunction=&MF_BlendLandscapeLayers}
  FunctionOutput(None=@n25.BlendedMaterial) {outputName=Material, sortPriority=-1}
  n26 = Multiply(A=MaterialFunctionCall() {materialFunction=/Game/Athena/Apollo/Environments/Landscape/Materials/Biome/CommonMaterials/MF_Common_Landscape_Layers}.Result) {constB=0}
  n27 = Constant()
  FunctionOutput(None=Add(
    A=@n25.Resulting Height Map
    B=FeatureLevelSwitch(Default=@n26, ES2=@n27, ES3_1=@n27, SM4_REMOVED=@n26, SM5=@n26, SM6=@n26) {desc=This is required for the target layers to behave correctly in the landscape tool}
   )) {outputName=Heightmap}
  FunctionOutput(None=NamedRerouteDeclaration(None=@n11.Normals) {name=Gravel_Normals, nodeColor=<rgba 1 0.7808 0.4624 1>}) {outputName=GravelNormals, sortPriority=6}
  n28 = LandscapeLayerSample() {parameterName=Mud}
  n29 = LandscapeLayerSample() {parameterName=Gravel}
  FunctionOutput(None=LinearInterpolate(
    A=Reroute() {desc="Connect to \"Displacement\" output of MF_Common_Grass"}
    B=Reroute() {desc="Connect to \"Displacement\" output of MF_Common_Gravel"}
    Alpha=Max(A=@n28, B=@n29)
   )) {outputName=Displacement, sortPriority=11}
  n30 = Constant() {r=5}
  n31 = Constant() {r=-1}
  FunctionOutput(None=Saturate(None=LinearInterpolate(A=@n30, B=@n31, Alpha=@n12.Alpha))) {desc=Used in Papaya version, outputName=GravelMaskForGrassFoliage, sortPriority=10}
  n32 = NamedRerouteUsage()
  n33 = NamedRerouteUsage()
  n34 = NamedRerouteUsage()
  n35 = NamedRerouteUsage()
  FunctionOutput(None=Max(A=Max(A=@n32, B=@n33), B=Max(A=@n34, B=@n35))) {outputName=CommonLayerMaskForGrassFoliage, sortPriority=13}
  n36 = Max(
   A=Max(A=LandscapeLayerSample() {parameterName=Mud}, B=LandscapeLayerSample() {parameterName=Gravel})
   B=LandscapeLayerSample() {parameterName=Sand}
  )
  n37 = LandscapeLayerSample() {parameterName=Asphalt}
  n38 = ComponentMask(None=VertexNormalWS()) {b=true}
  n39 = OneMinus(None=Power(Base=Abs(None=@n38), Exp=Constant() {r=3}))
  n40 = Constant() {r=0.3}
  n41 = Max(
   A=Max(A=Max(A=@n36, B=@n37), B=Multiply(A=@n39, B=@n40))
   B=FunctionInput(Preview=Constant()) {bUsePreviewValueAsDefault=true, inputName=AdditionalGrassBorderColorMask, inputType=FunctionInput_Scalar, sortPriority=-8}
  )
  n42 = Constant() {r=2}
  n43 = NamedRerouteDeclaration(None=FunctionInput() {bUsePreviewValueAsDefault=true, inputName=BiomeMask-RGBA, inputType=FunctionInput_Vector4}) {name=BiomeMask-RGBA, nodeColor=<rgba 0 1 0.7422 1>}
  n44 = MaterialFunctionCall!(
   BiomeMask-RGBA=@n43
   Grass Localised Tint Mask=NamedRerouteDeclaration(None=FunctionInput() {bUsePreviewValueAsDefault=true, inputName=Custom Biome Mask, inputType=FunctionInput_Scalar}) {name=Custom Biome Mask, nodeColor=<rgba 0.125 0 1 1>}
   CombinedLandscapeLayersMask=Saturate(None=Multiply(A=@n41, B=@n42))
   BiomeColors-UseDefault
  ) {materialFunction=&/MF_Asteria_Common_Grass}
  FunctionOutput(None=@n44.GrassColorMask1) {outputName=GrassColorMask1}
  FunctionOutput(None=@n44.GrassColorMask2) {outputName=GrassColorMask2}
  FunctionOutput(None=NamedRerouteDeclaration(None=@n44.GrassColorMask3) {name=GrassColorMask3, nodeColor=<rgba 1 0 0.9375 1>}) {outputName=GrassColorMask3}
  FunctionOutput(None=@n11.Material) {outputName=Gravel-Material}
  FunctionOutput(None=@n11.Height-W/Variance) {outputName=GravelHeightVariance}
  n45 = LinearInterpolate(A=Constant() {r=5}, B=Constant() {r=-1}, Alpha=@n14.Alpha)
  FunctionOutput(None=NamedRerouteDeclaration(None=Saturate(None=@n45)) {name=GM-Mud, nodeColor=<rgba 0.5 0 1 1>}) {outputName=GM-Mud, sortPriority=10}
  FunctionOutput(None=NamedRerouteDeclaration(None=@n11.Height) {name=Gravel_Height, nodeColor=<rgba 1 0.7808 0.4624 1>}) {outputName=GravelHeight, sortPriority=2}
  n46 = MaterialFunctionCall(
   Material 1=NamedRerouteUsage()
   Height Texture 1=NamedRerouteUsage()
   Material 2=NamedRerouteUsage()
   Height Texture 2=NamedRerouteUsage()
   Transistion Phase=NamedRerouteUsage()
  ) {materialFunction=&MF_BlendLandscapeLayers}
  FunctionOutput(None=NamedRerouteDeclaration(None=@n46.Alpha) {name=GM-Snow, nodeColor=<rgba 0.2422 1 0 1>}) {outputName=SnowScreeMask}
  n47 = Constant() {r=30}
  n48 = Constant()
  n49 = Add(A=@n5, B=Saturate(None=LinearInterpolate(A=@n47, B=@n48, Alpha=@n6.Alpha)))
  FunctionOutput(None=NamedRerouteDeclaration(None=Saturate(None=@n49)) {name=GM-Sand, nodeColor=<rgba 0.5 0 1 1>}) {outputName=GM-Sand, sortPriority=10}
  n50 = LinearInterpolate(A=Constant() {r=5}, B=Constant() {r=-1}, Alpha=@n25.Alpha)
  FunctionOutput(None=NamedRerouteDeclaration(None=Saturate(None=@n50)) {name=GM-Asphalt, nodeColor=<rgba 0.5 0 1 1>}) {outputName=GM-Asphalt, sortPriority=10}
  FunctionOutput(None=@n44.Default-LandscapeHeightmap) {outputName=GrassHeightmap, sortPriority=7}
  FunctionOutput(None=@n44.GrassColorMask4) {outputName=GrassNoiseMask, sortPriority=9}
  FunctionOutput(None=@n44.TangentNormals) {outputName=GrassNormal, sortPriority=8}
  FunctionOutput(None=@n44.BasecolorDetail) {outputName=GrassBaseColorDetail, sortPriority=10}
  FunctionOutput(None=NamedRerouteDeclaration(None=@n11.Diffuse) {name=Gravel_Diffuse, nodeColor=<rgba 1 0.7808 0.4624 1>}) {outputName=GravelDiffuse, sortPriority=3}
  FunctionOutput(None=NamedRerouteDeclaration(None=@n11.Spec) {name=Gravel_Spec, nodeColor=<rgba 1 0.7808 0.4624 1>}) {outputName=GravelSpec, sortPriority=4}
  FunctionOutput(None=NamedRerouteDeclaration(None=@n11.Rough) {name=Gravel_Roughness, nodeColor=<rgba 1 0.7808 0.4624 1>}) {outputName=GravelRough, sortPriority=5}
  Multiply(A=OneMinus(None=NamedRerouteUsage()))
  NamedRerouteDeclaration(None=FunctionInput(Preview=StaticBool()) {bUsePreviewValueAsDefault=true, inputName=PutSandBeforeGravel, inputType=FunctionInput_StaticBool}) {name=PutSandBeforeGravel, nodeColor=<rgba 0.5 0 1 1>}
  NamedRerouteDeclaration(None=@n44.GrassColorMask1) {name=GrassColorMask1, nodeColor=<rgba 0.5 0 1 1>}
  n51 = MaterialFunctionCall!(
   Grassmask
   Gravel-Height
   CommonLayers-Height=NamedRerouteUsage()
   GrassColorInput=NamedRerouteUsage()
   DetailMasking=NamedRerouteUsage()
   InputMaterialHeight
   Material=NamedRerouteUsage()
  ) {materialFunction=&/MF_FarmPatches}
  NamedRerouteDeclaration(None=@n51.Result) {name=FarmPatchMaterial, nodeColor=<rgba 0.0312 1 0 1>}
  NamedRerouteDeclaration(None=@n44.BasecolorDetail) {name=BaseColorDetail, nodeColor=<rgba 0 0.4141 1 1>}
  NamedRerouteDeclaration(None=@n44.DistanceMasking) {name=DistanceMasking, nodeColor=<rgba 0 1 0.7891 1>}
  NamedRerouteDeclaration(None=Constant()) {name=PathMask, nodeColor=<rgba 1 0 0.7969 1>}
  n52 = Max(A=LandscapeLayerSample() {parameterName=Asphalt}, B=LandscapeLayerSample() {parameterName=Mud})
  n53 = LandscapeLayerSample() {parameterName=Sand}
  n54 = LandscapeLayerSample() {parameterName=Gravel}
  NamedRerouteDeclaration(None=Max(A=@n52, B=Max(A=@n53, B=@n54))) {name=CommonLayerMaskForGrassFoliage, nodeColor=<rgba 0 1 0.6953 1>}
  n55 = Multiply(
   A=ComponentMask(None=MaterialFunctionCall!(
     Texture
     LandscapeCoordinates=@n1
     LowEndTiling
     HexGridGradientSharpness
     HexGridSize
     HexGridUVInput
     MinScale
     MaxScale
     MinOffset
     MaxOffset
     "MaxRotation(Degrees)"
     "MinRotation(Degrees)"
     forceLowQuality
    ) {materialFunction=/Game/Athena/Asteria/Environments/Landscape/Materials/MaterialFunctions/UtilityFunctions/Landscape_TextureSample}.Result) {a=true}
   B=ScalarParameter() {defaultValue=1.75, parameterName=GravelNoiseStrength}
  )
  n56 = ScalarParameter() {defaultValue=2, parameterName=GravelNoisePow}
  NamedRerouteDeclaration(None=Clamp(None=Power(Base=@n55, Exp=@n56))) {name=GrassDirtMask, nodeColor=<rgba 0.8984 0 1 1>}
  NamedRerouteDeclaration(None=@n44.Default-LandscapeMaterial) {name=GrassMaterial, nodeColor=<rgba 0.5 0 1 1>}
  NamedRerouteDeclaration(None=@n51.Height) {name=FarmPatchHeight, nodeColor=<rgba 1 0.6563 0 1>}
  NamedRerouteDeclaration(None=FunctionInput(Preview=SetMaterialAttributes!(MaterialAttributes)) {bUsePreviewValueAsDefault=true, inputName=CustombaseLayerMaterial, inputType=FunctionInput_MaterialAttributes}) {name=CustomBaseMaterial, nodeColor=<rgba 1 0.3516 0 1>}
  NamedRerouteDeclaration(None=FunctionInput(Preview=Constant()) {bUsePreviewValueAsDefault=true, inputName=CustombaseLayerHeight, inputType=FunctionInput_Scalar}) {name=CustomBaseMaterialHeight, nodeColor=<rgba 1 0 0.5391 1>}
  n57 = NamedRerouteUsage()
  NamedRerouteDeclaration(None=StaticSwitch(True=@n46.BlendedMaterial, False=NamedRerouteUsage(), Value=@n57)) {name=BaseLayerMaterial, nodeColor=<rgba 0.0547 0 1 1>}
  NamedRerouteDeclaration(None=StaticSwitch(True=@n46.Resulting Height Map, False=NamedRerouteUsage(), Value=@n57)) {name=BaseLayerHeightMap, nodeColor=<rgba 0 1 0.6953 1>}
  NamedRerouteDeclaration(None=StaticBoolParameter() {defaultValue=true, parameterName=UseCustomBaseLayerMaterial}) {name=UseCustomBaseLayerMaterial, nodeColor=<rgba 0 1 0.1094 1>}
  NamedRerouteDeclaration(None=ChannelMaskParameter(None=@n43) {parameterName=BiomeMaskTundra}.R) {name=Layer-Snow, nodeColor=<rgba 1 1 1 1>}
  NamedRerouteDeclaration(None=OneMinus(None=Max(A=@n52, B=@n53))) {name=CommonLayerMaskforGravelScree, nodeColor=<rgba 1 0 0.75 1>}
  NamedRerouteDeclaration(None=@n44.Default-LandscapeHeightmap) {name=GrassHeight, nodeColor=<rgba 0.5 0 1 1>}
  n58 = Constant() {r=5}
  n59 = Constant() {r=-1}
  NamedRerouteDeclaration(None=Saturate(None=LinearInterpolate(A=@n58, B=@n59, Alpha=@n12.Alpha))) {name=GM-Gravel, nodeColor=<rgba 0.5 0 1 1>}
  NamedRerouteUsage()
  NamedRerouteUsage()
  ScalarParameter() {parameterName=FeatureZeroOut}
  n60 = GetMaterialAttributes(None=@n25.BlendedMaterial) {attributeGetTypes [#1]}
  SetMaterialAttributes!(
   MaterialAttributes=@n60.MaterialAttributes
   BaseColor=LinearInterpolate(A=@n60.BaseColor, B=@n25.Resulting Height Map, Alpha=ScalarParameter() {parameterName=debug-ts243rwrt})
  ) {attributeSetTypes [#1]}
  n61 = Constant() {r=1}
  n62 = ScalarParameter() {defaultValue=1, parameterName=GravelMask-Spread}
  n63 = ScalarParameter() {parameterName=GravelMask-Hardness}
  SphereMask(A=@n12.Alpha, B=@n61, Radius=@n62, Hardness=@n63)
  SphereMask(A=@n12.Alpha, B=@n61, Radius=@n62, Hardness=@n63)
 ]}
