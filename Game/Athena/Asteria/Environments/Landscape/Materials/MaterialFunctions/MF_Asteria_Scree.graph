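# Landscape layer material function for Asteria scree, outputting material attributes, height and a base layer mask.
# Tiled diffuse, normal and packed mask textures supply color, normal, specular and roughness, with height taken from diffuse alpha.
# MF_SnowBuildup uses that height to produce snow color, blended by the Tundra biome mask input, and a coverage mask.
# Hex-grid randomized sampling and rock-detail blend branches are present but do not feed the outputs.
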MaterialFunction /Game/Athena/Asteria/Environments/Landscape/Materials/MaterialFunctions/MF_Asteria_Scree
graph {nodes [
  Constant() {r=0.2}
  Constant() {r=1}
  n1 = NamedRerouteUsage()
  n2 = SetMaterialAttributes!(
   MaterialAttributes=GetMaterialAttributes().MaterialAttributes
   BaseColor=NamedRerouteUsage()
   Specular=NamedRerouteUsage()
   Roughness=NamedRerouteUsage()
   Normal=@n1
  ) {attributeSetTypes [
    #1
    #2
    #3
    #4
   ]}
  FunctionOutput(None=@n2) {outputName=Material}
  FunctionOutput(None=NamedRerouteUsage()) {outputName=BaseLayerMask, sortPriority=2}
  FunctionOutput(None=NamedRerouteUsage()) {outputName=Height, sortPriority=1}
  MaterialFunctionCall!(
   Texture=NamedRerouteUsage()
   LandscapeCoordinates=NamedRerouteUsage()
   LowEndTiling
   HexGridGradientSharpness=NamedRerouteUsage()
   HexGridSize
   HexGridUVInput=NamedRerouteUsage()
   MinScale=NamedRerouteUsage()
   MaxScale=NamedRerouteUsage()
   MinOffset=NamedRerouteUsage()
   MaxOffset=NamedRerouteUsage()
   "MaxRotation(Degrees)"=NamedRerouteUsage()
   "MinRotation(Degrees)"=NamedRerouteUsage()
   forceLowQuality
  ) {materialFunction=/Game/Athena/Asteria/Environments/Landscape/Materials/MaterialFunctions/UtilityFunctions/Landscape_TextureSample}
  MaterialFunctionCall(
   Texture=NamedRerouteUsage()
   LandscapeCoordinates=NamedRerouteUsage()
   HexGridGradientSharpness=NamedRerouteUsage()
   HexGridUVInput=NamedRerouteUsage()
   MinScale=NamedRerouteUsage()
   MaxScale=NamedRerouteUsage()
   MinOffset=NamedRerouteUsage()
   MaxOffset=NamedRerouteUsage()
   "MaxRotation(Degrees)"=NamedRerouteUsage()
   "MinRotation(Degrees)"=NamedRerouteUsage()
  ) {materialFunction=&Landscape_TextureSample}
  n3 = Multiply(A=LandscapeLayerCoords() {mappingScale=1}, B=ScalarParameter() {defaultValue=0.2, parameterName=LS-ScreeDetailScale})
  n4 = TextureSample(UVs=@n3) {
   samplerSource=SSM_Wrap_WorldGroupSettings
   samplerType=SAMPLERTYPE_Normal
   texture=/Game/Environments/Apollo/Rocks/Natural_Rocks/Textures/T_RockDetails_02_N
  }
  MaterialFunctionCall(AdditionalNormal=@n4.RGB) {materialFunction=&BlendAngleCorrectedNormals}
  MaterialFunctionCall(BaseNormal=@n1, AdditionalNormal=NamedRerouteUsage()) {materialFunction=&BlendAngleCorrectedNormals}
  n5 = MaterialFunctionCall(Float3=MaterialFunctionCall!(
    Texture=NamedRerouteUsage()
    UseMipBias
    MipBias
    LowEndTiling
    LandscapeCoordinates=NamedRerouteUsage()
    HexGridSize
    HexGridUVInput=NamedRerouteUsage()
    HexGridGradientSharpness=NamedRerouteUsage()
    MaxScale=NamedRerouteUsage()
    MinScale=NamedRerouteUsage()
    MaxOffset=NamedRerouteUsage()
    MinOffset=NamedRerouteUsage()
    "MaxRotation(Degrees)"=NamedRerouteUsage()
    "MinRotation(Degrees)"=NamedRerouteUsage()
    forceLowQuality
   ) {materialFunction=/Game/Athena/Asteria/Environments/Landscape/Materials/MaterialFunctions/UtilityFunctions/Landscape_NormalTextureSample}.Result) {materialFunction=&BreakOutFloat3Components}
  MaterialFunctionCall(X=@n5.R, Y=OneMinus(None=@n5.G), Z=@n5.R) {materialFunction=&MakeFloat3}
  n6 = TextureSample(UVs=@n3) {samplerSource=SSM_Wrap_WorldGroupSettings, texture=/Game/Environments/Apollo/Rocks/Natural_Rocks/Textures/T_RockDetails_02D}
  n7 = Add(A=Constant() {r=-1.1}, B=Multiply(A=Constant() {r=2.05}, B=@n6.R))
  n8 = LinearInterpolate(A=Constant() {r=7}, B=Constant() {r=-0.8}, Alpha=@n6.R)
  n9 = LinearInterpolate(A=Constant() {r=-0.25}, B=Constant() {r=1.5}, Alpha=Reroute())
  Multiply(A=Multiply(A=Saturate(None=@n7), B=Saturate(None=@n8)), B=Clamp(None=Saturate(None=@n9), Max=Constant() {r=0.5}) {clampMode=CMODE_ClampMax})
  Multiply(A=@n4.RGB, B=VectorParameter() {parameterName=LS-Scree-DetailNormal, defaultValue=<rgba 1 1 1 1>}.RGB)
  NamedRerouteDeclaration(None=TextureObject() {texture=/Game/Environments/Apollo/Sets/Terrain/Textures/T_Apollo_Scree_D_01}) {name=Diffuse_Source, nodeColor=<rgba 0 1 0.0391 1>}
  NamedRerouteDeclaration(None=TextureObject() {samplerType=SAMPLERTYPE_Normal, texture=/Game/Environments/Apollo/Sets/Terrain/Textures/T_Apollo_Scree_N_01}) {name=Normal_Source, nodeColor=<rgba 0 1 0.0156 1>}
  n10 = Multiply(A=LandscapeLayerCoords() {mappingScale=1}, B=ScalarParameter() {defaultValue=0.25, parameterName=ScreeTiling})
  n11 = TextureSampleParameter2D(UVs=@n10) {
   group=Scree
   parameterName=Scree_spec
   samplerSource=SSM_Wrap_WorldGroupSettings
   samplerType=SAMPLERTYPE_Masks
   texture=/Game/Environments/Apollo/Sets/Terrain/Textures/T_Apollo_Scree_S_01
  }
  n12 = MaterialFunctionCall(Float3=@n11.RGB) {materialFunction=&BreakOutFloat3Components}
  NamedRerouteDeclaration(None=@n12.B) {name=Roughness, nodeColor=<rgba 1 0 0.3281 1>}
  NamedRerouteDeclaration(None=ScalarParameter() {parameterName=Scree_MinRotation}) {name=MinRotation, nodeColor=<rgba 0 1 0.8359 1>}
  NamedRerouteDeclaration(None=ScalarParameter() {parameterName=Scree_MaxRotation}) {name=MaxRotation, nodeColor=<rgba 0.6172 0 1 1>}
  NamedRerouteDeclaration(None=TextureObject() {samplerType=SAMPLERTYPE_Normal, texture=&T_RockDetails_02_N}) {name=RockDetails_N_Source, nodeColor=<rgba 1 0.0469 0 1>}
  n13 = MaterialFunctionCall!(Scale=ScalarParameter() {defaultValue=400, parameterName=ScreeScale}) {materialFunction=/Game/Athena/Apollo/Environments/Landscape/Materials/Biome/CommonMaterials/MF_SimpleWorldUV}
  NamedRerouteDeclaration(None=Multiply(A=@n13.Result, B=ScalarParameter() {defaultValue=0.5, parameterName=HexGridUVMultiplier})) {name=HexGridUVScale, nodeColor=<rgba 0 1 0.7188 1>}
  n14 = Constant() {r=1}
  NamedRerouteDeclaration(None=MaterialFunctionCall(
    Texture=NamedRerouteUsage()
    LandscapeCoordinates=NamedRerouteUsage()
    HexGridUVInput=NamedRerouteUsage()
    HexGridGradientSharpness=NamedRerouteUsage()
    MaxScale=@n14
    MinScale=@n14
    MaxOffset=NamedRerouteUsage()
    MinOffset=NamedRerouteUsage()
    "MaxRotation(Degrees)"=NamedRerouteUsage()
    "MinRotation(Degrees)"=NamedRerouteUsage()
   ) {materialFunction=&Landscape_NormalTextureSample}.Result) {name=DetailNormals, nodeColor=<rgba 0 0.2969 1 1>}
  NamedRerouteDeclaration(None=ScalarParameter() {defaultValue=6, parameterName=HexGridGradientSharpness}) {name=HexGradientSharpness, nodeColor=<rgba 0.4766 0 1 1>}
  NamedRerouteDeclaration(None=MaterialFunctionCall(Scale=ScalarParameter() {defaultValue=200, parameterName=ScreeDetailScale}) {materialFunction=&MF_SimpleWorldUV}.Result) {name=DetailSCale, nodeColor=<rgba 1 0 0.3281 1>}
  NamedRerouteDeclaration(None=ScalarParameter() {defaultValue=-1, parameterName=Scree_MinOffset}) {name=MinOffset, nodeColor=<rgba 0.0781 0 1 1>}
  NamedRerouteDeclaration(None=TextureObject() {samplerType=SAMPLERTYPE_Masks, texture=&T_Apollo_Scree_S_01}) {name=Spec_Source, nodeColor=<rgba 0 1 0.0156 1>}
  NamedRerouteDeclaration(None=ScalarParameter() {defaultValue=1, parameterName=Scree_MaxOffset}) {name=MaxOffset, nodeColor=<rgba 1 0.6094 0 1>}
  NamedRerouteDeclaration(None=@n13.Result) {name=BaseScale, nodeColor=<rgba 1 0.4688 0 1>}
  NamedRerouteDeclaration(None=ScalarParameter() {defaultValue=1, parameterName=Scree_MinScale}) {name=MinScale, nodeColor=<rgba 0.9922 1 0 1>}
  NamedRerouteDeclaration(None=ScalarParameter() {defaultValue=2, parameterName=Scree_MaxScale}) {name=MaxScale, nodeColor=<rgba 0 0.8125 1 1>}
  n15 = TextureSampleParameter2D(UVs=@n10) {
   group=Scree
   parameterName=Scree_Diffuse
   samplerSource=SSM_Wrap_WorldGroupSettings
   texture=&T_Apollo_Scree_D_01
  }
  n16 = FunctionInput(Preview=Constant()) {bUsePreviewValueAsDefault=true, inputName=BiomeMask-Tundra, inputType=FunctionInput_Scalar}
  n17 = MaterialFunctionCall!(
   Height=NamedRerouteDeclaration(None=ComponentMask(None=@n15.RGBA) {a=true}) {name=Height, nodeColor=<rgba 1 0.4219 0 1>}
   HeightSampleRadius
   HeightSampleHardness
   FlipHeightMap
   SnowColorBlend
   BiomeMask-Tundra=@n16
   Diffuse=@n15.RGB
  ) {materialFunction=&/MF_SnowBuildup}
  NamedRerouteDeclaration(None=@n17.SnowCoverageMask) {name=ScreeBaseLayerMask, nodeColor=<rgba 0.5469 0 1 1>}
  NamedRerouteDeclaration(None=LinearInterpolate(A=@n15.RGB, B=@n17.Diffuse, Alpha=@n16)) {name=Diffuse, nodeColor=<rgba 0 0.5312 1 1>}
  NamedRerouteDeclaration(None=@n12.R) {name=Specular, nodeColor=<rgba 0 1 0.6016 1>}
  n18 = TextureSampleParameter2D(UVs=@n10) {
   group=Scree
   parameterName=Scree_Normal
   samplerSource=SSM_Wrap_WorldGroupSettings
   samplerType=SAMPLERTYPE_Normal
   texture=&T_Apollo_Scree_N_01
  }
  NamedRerouteDeclaration(None=@n18.RGB) {name=Normal, nodeColor=<rgba 1 0 0.0703 1>}
  NamedRerouteUsage()
  NamedRerouteUsage()
  n19 = SquareRoot(None=@n6.RGB)
  Saturate(None=Multiply(A=Constant() {r=1.5}, B=Multiply(B=@n19)))
 ]}
