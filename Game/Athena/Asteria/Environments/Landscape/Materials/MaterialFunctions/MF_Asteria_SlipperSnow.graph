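# MF_Asteria_SlipperSnow is a landscape material function that supplies the slippery snow layer for Asteria terrain.
# Tiled landscape coordinates drive diffuse, packed specular mask and normal texture samples, all exposed as parameters.
# The mask's red and blue channels blend between min/max specular and roughness parameters, feeding a material attributes output with zero metallic.
# The diffuse alpha is exposed as a separate Height output.
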MaterialFunction /Game/Athena/Asteria/Environments/Landscape/Materials/MaterialFunctions/MF_Asteria_SlipperSnow
graph {nodes [
  n1 = Multiply(A=LandscapeLayerCoords(), B=ScalarParameter() {defaultValue=0.2, group=SlipperySnow, parameterName=SL_Snow_Tiling})
  n2 = TextureSampleParameter2D(UVs=@n1) {group=SlipperySnow, parameterName=SL_Diffuse, texture=/Game/Environments/Asteria/Terrain/Textures/T_Asteria_SlipperySnow_01_D}
  n3 = ScalarParameter() {defaultValue=0.5, group=SlipperySnow, parameterName=SL_PRVT_MinSpec}
  n4 = ScalarParameter() {defaultValue=1, group=SlipperySnow, parameterName=SL_PRVT_MaxSpec}
  n5 = TextureSampleParameter2D(UVs=@n1) {
   group=SlipperySnow
   parameterName=SL_Specular
   samplerType=SAMPLERTYPE_Masks
   texture=/Game/Environments/Asteria/Terrain/Textures/T_Asteria_SlipperySnow_01_S
  }
  n6 = ScalarParameter() {defaultValue=0.2, group=SlipperySnow, parameterName=SL_PRVT_MinRoughness}
  n7 = ScalarParameter() {defaultValue=0.4, group=SlipperySnow, parameterName=SL_PRVT_MaxRoughness}
  n8 = TextureSampleParameter2D(UVs=@n1) {
   group=SlipperySnow
   parameterName=SL_Normal
   samplerType=SAMPLERTYPE_Normal
   texture=/Game/Environments/Asteria/Terrain/Textures/T_Asteria_SlipperySnow_01_N
  }
  FunctionOutput(None=MakeMaterialAttributes!(
    BaseColor=@n2.RGB
    Metallic=Constant()
    Specular=LinearInterpolate(A=@n3, B=@n4, Alpha=@n5.R)
    Roughness=LinearInterpolate(A=@n6, B=@n7, Alpha=@n5.B)
    Anisotropy
    EmissiveColor
    Opacity
    OpacityMask
    Normal=@n8.RGB
    Tangent
    WorldPositionOffset
    SubsurfaceColor
    ClearCoat
    ClearCoatRoughness
    AmbientOcclusion
    Refraction
    CustomizedUVs_0
    CustomizedUVs_1
    CustomizedUVs_2
    CustomizedUVs_3
    CustomizedUVs_4
    CustomizedUVs_5
    CustomizedUVs_6
    CustomizedUVs_7
    PixelDepthOffset
    ShadingModel
    Displacement
   ))
  FunctionOutput(None=@n2.A) {outputName=Height, sortPriority=1}
 ]}
