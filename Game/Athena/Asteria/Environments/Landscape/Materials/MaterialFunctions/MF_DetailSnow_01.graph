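# Landscape material function for a wind-blown detail snow layer in the Asteria environment, outputting material attributes and a height value.
# Tiled landscape coordinates and hex-grid parameters drive three Landscape_TextureSample calls that vary scale, with offset and rotation zeroed.
# Base color comes solely from MF_SnowColorBase, while texture samples supply normal, height from alpha, and one blue channel shared by roughness and specular.
# An albedo-times-texture tint is computed but left unconnected.
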MaterialFunction /Game/Athena/Asteria/Environments/Landscape/Materials/MaterialFunctions/MF_DetailSnow_01
graph {nodes [
  Constant3Vector() {constant=<rgba 0.26 0.3355 0.4 1>}
  n1 = MaterialFunctionCall() {materialFunction=&/MF_SnowColorBase}
  n2 = NamedRerouteUsage()
  n3 = NamedRerouteUsage()
  n4 = Constant()
  n5 = MaterialFunctionCall(Float3=MaterialFunctionCall!(
    Texture=NamedRerouteUsage()
    LandscapeCoordinates=@n2
    LowEndTiling
    HexGridGradientSharpness
    HexGridSize=NamedRerouteUsage()
    HexGridUVInput=Multiply(A=@n2, B=@n3)
    MinScale=NamedRerouteUsage()
    MaxScale=NamedRerouteUsage()
    MinOffset=@n4
    MaxOffset=@n4
    "MaxRotation(Degrees)"=@n4
    "MinRotation(Degrees)"=@n4
    forceLowQuality
   ) {materialFunction=/Game/Athena/Asteria/Environments/Landscape/Materials/MaterialFunctions/UtilityFunctions/Landscape_TextureSample}.Result) {materialFunction=&BreakOutFloat3Components}
  n6 = NamedRerouteUsage()
  n7 = NamedRerouteUsage()
  n8 = Constant()
  n9 = SetMaterialAttributes!(
   MaterialAttributes
   BaseColor=@n1.SnowAlbedo
   Roughness=@n5.B
   Specular=@n5.B
   Normal=MaterialFunctionCall(
    Texture=NamedRerouteUsage()
    LandscapeCoordinates=@n6
    HexGridSize=NamedRerouteUsage()
    HexGridUVInput=Multiply(A=@n6, B=@n7)
    MinScale=NamedRerouteUsage()
    MaxScale=NamedRerouteUsage()
    MinOffset=@n8
    MaxOffset=@n8
    "MaxRotation(Degrees)"=@n8
    "MinRotation(Degrees)"=@n8
   ) {materialFunction=&Landscape_TextureSample}.Result
  ) {attributeSetTypes [
    #1
    #2
    #3
    #4
   ]}
  FunctionOutput(None=@n9) {outputName=Material}
  n10 = NamedRerouteUsage()
  n11 = Constant()
  n12 = MaterialFunctionCall(
   Texture=NamedRerouteUsage()
   LandscapeCoordinates=@n10
   HexGridSize=NamedRerouteUsage()
   HexGridUVInput=Multiply(A=@n10, B=NamedRerouteUsage())
   MinScale=NamedRerouteUsage()
   MaxScale=NamedRerouteUsage()
   MinOffset=@n11
   MaxOffset=@n11
   "MaxRotation(Degrees)"=@n11
   "MinRotation(Degrees)"=@n11
  ) {materialFunction=&Landscape_TextureSample}
  FunctionOutput(None=ComponentMask(None=@n12.Result) {a=true}) {outputName=Height}
  Multiply(A=@n1.SnowAlbedo, B=ComponentMask(None=@n12.Result) {b=true, g=true, r=true})
  NamedRerouteDeclaration(None=Multiply(A=LandscapeLayerCoords() {mappingScale=1}, B=ScalarParameter() {defaultValue=0.25, group=DetailSnow_01, parameterName=DetailSnow_Tiling})) {name=UVCoords, nodeColor=<rgba 0 1 0.2969 1>}
  NamedRerouteDeclaration(None=TextureObjectParameter() {parameterName=DetailSnow_D, texture=/Game/Environments/Asteria/Terrain/Textures/T_Asteria_WindBlownSnow_02_D}) {name=Diffuse, nodeColor=<rgba 0 1 0.7656 1>}
  n13 = TextureObjectParameter() {parameterName=DetailSnow_S, samplerType=SAMPLERTYPE_Masks, texture=/Game/Environments/Asteria/Terrain/Textures/T_Asteria_WindBlownSnow_02_S}
  NamedRerouteDeclaration(None=@n13) {name=Specular, nodeColor=<rgba 0 1 0.1562 1>}
  n14 = TextureObjectParameter() {parameterName=DetailSnow_N, samplerType=SAMPLERTYPE_Normal, texture=/Game/Environments/Asteria/Terrain/Textures/T_Asteria_WindBlownSnow_02_N}
  NamedRerouteDeclaration(None=@n14) {name=Normal, nodeColor=<rgba 0 0.7656 1 1>}
  NamedRerouteDeclaration(None=ScalarParameter() {defaultValue=1, group=DetailSnow_01, parameterName=DS_MinScale}) {name=MinScale, nodeColor=<rgba 0.0547 0 1 1>}
  NamedRerouteDeclaration(None=ScalarParameter() {defaultValue=2, group=DetailSnow_01, parameterName=DS_MaxScale}) {name=MaxScale, nodeColor=<rgba 0.1016 1 0 1>}
  NamedRerouteDeclaration(None=ScalarParameter() {defaultValue=1, group=DetailSnow_01, parameterName=DS_HexGridUV_Scale}) {name=HexGridUVScale, nodeColor=<rgba 0.7813 0 1 1>}
  NamedRerouteDeclaration(None=ScalarParameter() {defaultValue=1, group=DetailSnow_01, parameterName=HS_HexGridSize}) {name=HexGridSize, nodeColor=<rgba 1 0.3516 0 1>}
 ]}
