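# Material function for Asteria landscapes that layers post-runtime-virtual-texture details, namely snow and Fortanium veins, onto an input material.
# It first overrides shading model, subsurface color and opacity, with default-lit parameters for the latter two.
# Static switches then optionally blend in MF_SnowPostRVT output and MF_FortaniumVeins output by mask, the vein mask boosted and clamped.
# An MF_BlendLandscapeLayers call is present but its output is unconnected.
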MaterialFunction /Game/Athena/Asteria/Environments/Landscape/Materials/MaterialFunctions/MF_PostRVTDetails
graph {nodes [
  Constant() {r=1}
  n1 = GetMaterialAttributes(None=FunctionInput() {inputName=Material, inputType=FunctionInput_MaterialAttributes})
  n2 = SetMaterialAttributes!(
   MaterialAttributes=GetMaterialAttributes(None=@n1.MaterialAttributes).MaterialAttributes
   ShadingModel=ShadingModel()
   SubsurfaceColor=ScalarParameter() {parameterName=DefaultLit_SSColor}
   Opacity=ScalarParameter() {defaultValue=1, parameterName=DefaultLit_Opacity}
  ) {attributeSetTypes [#1, #2, #3]}
  n3 = Constant()
  n4 = MaterialFunctionCall!(
   Mask=NamedRerouteDeclaration(None=FunctionInput(Preview=@n3) {bUsePreviewValueAsDefault=true, inputName=SnowMask, inputType=FunctionInput_Scalar, sortPriority=1}) {name=SnowMask, nodeColor=<rgba 1 0.4219 0 1>}
   SlipperySnow_Mask=NamedRerouteUsage()
   "Material;"=@n1.MaterialAttributes
  ) {materialFunction=&/MF_SnowPostRVT}
  n5 = NamedRerouteUsage()
  n6 = StaticSwitchParameter(True=BlendMaterialAttributes(A=@n2, B=@n4.SnowMaterial, Alpha=@n5) {vertexAttributeBlendType=UseA}, False=@n2) {parameterName=UsePostRVTSnow}
  n7 = NamedRerouteUsage()
  n8 = BlendMaterialAttributes(
   A=@n6
   B=MaterialFunctionCall!(VeinMask=@n7, Material=@n6, Position) {materialFunction=&/MF_FortaniumVeins}.Result
   Alpha=Clamp(None=Multiply(A=@n7) {constB=4})
  )
  FunctionOutput(None=SetMaterialAttributes!(MaterialAttributes=StaticSwitchParameter(True=@n8, False=@n6) {parameterName=UseFortaniumVeins}))
  MaterialFunctionCall!(
   Material 1=@n2
   Height Texture 1=NamedRerouteUsage()
   Material 2=@n4.SnowMaterial
   Height Texture 2=@n4.SnowMaterial
   Transistion Phase=@n5
   UseContrast
   Contrast
   AddLayerMask
  ) {materialFunction=/Game/Athena/Apollo/Environments/Landscape/Materials/Biome/CommonMaterials/MF_BlendLandscapeLayers}
  NamedRerouteDeclaration(None=FunctionInput(Preview=Constant()) {bUsePreviewValueAsDefault=true, inputName=SL_Snow_Mask, inputType=FunctionInput_Scalar, sortPriority=3}) {name=SL_SnowMask, nodeColor=<rgba 0.1484 1 0 1>}
  NamedRerouteDeclaration(None=FunctionInput(Preview=Constant()) {bUsePreviewValueAsDefault=true, inputName=CliffVeinsMask, inputType=FunctionInput_Scalar, sortPriority=2}) {name=CliffVeinsMask, nodeColor=<rgba 0.6406 0 1 1>}
  NamedRerouteDeclaration(None=FunctionInput(Preview=Constant()) {bUsePreviewValueAsDefault=true, inputName=Height_Input, inputType=FunctionInput_Scalar}) {name=Height_Input, nodeColor=<rgba 0.4297 0 1 1>}
 ]}
