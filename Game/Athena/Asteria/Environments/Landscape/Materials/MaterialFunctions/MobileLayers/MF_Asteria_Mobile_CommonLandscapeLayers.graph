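# Mobile landscape material function for Asteria that composites the common terrain layers into one material.
# Height-aware blends driven by landscape layer weights stack grass, gravel, mud, asphalt or jungle mud, and cliff, with sand or jungle floor switchable before or after gravel.
# Static switches add ice from the BiomeB layer and snow on top, with a snow mask derived from biome red minus other layers.
# A shadow-tint and height-darkening branch is left unconnected to the output.
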MaterialFunction /Game/Athena/Asteria/Environments/Landscape/Materials/MaterialFunctions/MobileLayers/MF_Asteria_Mobile_CommonLandscapeLayers
graph {nodes [
  Constant()
  Constant()
  FunctionInput(Preview=StaticBool()) {bUsePreviewValueAsDefault=true, inputName=PutSandBeforeGravel, inputType=FunctionInput_StaticBool}
  n1 = NamedRerouteUsage()
  n2 = NamedRerouteUsage()
  n3 = MaterialFunctionCall!(MudMaterial-Input=NamedRerouteUsage()) {materialFunction=&/MF_Mobile_Asteria_Ice}
  n4 = Saturate(None=Multiply(A=LandscapeLayerSample() {parameterName=BiomeB}) {constB=3})
  n5 = FunctionInput() {bUsePreviewValueAsDefault=true, inputName=SnowMask, inputType=FunctionInput_Scalar, previewValue=<1 0 0 1>}
  n6 = MaterialFunctionCall!(
   Material 1=@n1
   Height Texture 1=@n2
   Material 2=@n3.Result
   Height Texture 2=@n3.Result
   Transistion Phase=StaticSwitchParameter(True=Multiply(A=@n4, B=@n5), False=@n4) {parameterName=UseMobileSnowMask}
   UseContrast
   Contrast
   AddLayerMask
  ) {materialFunction=/Game/Athena/Apollo/Environments/Landscape/Materials/Biome/CommonMaterials/MF_BlendLandscapeLayers}
  n7 = MaterialFunctionCall() {materialFunction=&/MF_Mobile_Asteria_Snow}
  n8 = StaticSwitch(
   True=MaterialFunctionCall(
    Material 1=StaticSwitchParameter(True=@n6.BlendedMaterial, False=@n1) {parameterName=MobileUseIce}
    Height Texture 1=StaticSwitchParameter(True=@n6.Resulting Height Map, False=@n2) {parameterName=MobileUseIce}
    Material 2=@n7.Result
    Height Texture 2=@n7.Result
    Transistion Phase=NamedRerouteUsage()
   ) {materialFunction=&MF_BlendLandscapeLayers}.BlendedMaterial
   False=NamedRerouteUsage()
   Value=NamedRerouteUsage()
  )
  FunctionOutput(None=SetMaterialAttributes!(MaterialAttributes=GetMaterialAttributes(None=@n8).MaterialAttributes, ShadingModel=ShadingModel()) {attributeSetTypes [#1]}) {outputName=Material, sortPriority=-1}
  n9 = NamedRerouteUsage()
  Multiply(A=@n9, B=OneMinus(None=NamedRerouteUsage()))
  NamedRerouteDeclaration(None=StaticBoolParameter() {group=Mobile, parameterName=Mobile-PutSandBeforeGravel}) {name=PutSandBeforeGravel, nodeColor=<rgba 0.5 0 1 1>}
  NamedRerouteDeclaration(None=FunctionInput(Preview=StaticBool()) {bUsePreviewValueAsDefault=true, inputName=UseSnow, inputType=FunctionInput_StaticBool, sortPriority=1}) {name=UseSnow, nodeColor=<rgba 0 0.4609 1 1>}
  n10 = MaterialFunctionCall!(FallBackMobileMudGrayscale=NamedRerouteUsage()) {materialFunction=/Game/Athena/Asteria/Environments/Landscape/Materials/MaterialFunctions/MF_JungleGameplayMud}
  NamedRerouteDeclaration(None=@n10.MobileMaterial) {name=gameplayMudMob, nodeColor=<rgba 0.6406 1 0 1>}
  NamedRerouteDeclaration(None=@n7.Raw Height) {name=SnowHeight, nodeColor=<rgba 1 0 0.6797 1>}
  n11 = LandscapeLayerSample() {parameterName=Mud}
  n12 = LandscapeLayerSample() {parameterName=Gravel}
  n13 = Max(A=Max(A=Max(A=@n11, B=@n12), B=LandscapeLayerSample() {parameterName=Sand}), B=LandscapeLayerSample() {parameterName=Asphalt})
  n14 = Constant() {r=3}
  n15 = MaterialFunctionCall!(BiomeMask-RGBA=NamedRerouteUsage(), CombinedLandscapeLayersMask=Saturate(None=Multiply(A=@n13, B=@n14)), BypassUseRVT, BypassBiomeColors) {materialFunction=&/MF_Mobile_Asteria_Grass}
  n16 = MaterialFunctionCall() {materialFunction=/Game/Athena/Asteria/Environments/Landscape/Materials/MaterialFunctions/MF_JungleFloor}
  n17 = MaterialFunctionCall!(CustomSeaLevel, BiomeMask-RGBA=NamedRerouteUsage(), BypassBiomeColors) {materialFunction=&/MF_Mobile_Asteria_Sand}
  n18 = MaterialFunctionCall(
   Material 1=@n15.MobileLandscapeMaterial
   Height Texture 1=@n15.MobileLandscapeMaterial
   Material 2=StaticSwitchParameter(True=@n16.MobileMaterial, False=@n17.Material) {parameterName=UseJungleFloor}
   Height Texture 2=StaticSwitchParameter(True=@n16.Height, False=@n17.Height) {parameterName=UseJungleFloor}
   Transistion Phase=LandscapeLayerSample() {parameterName=Sand}
  ) {materialFunction=&MF_BlendLandscapeLayers}
  n19 = NamedRerouteUsage()
  n20 = NamedRerouteUsage()
  n21 = LandscapeLayerSample() {parameterName=Gravel}
  n22 = MaterialFunctionCall!(OverrideEdgeColor, BiomeMask-RGBA, EdgeMask=@n21, BypassBiomeColors) {materialFunction=&/MF_Mobile_Asteria_Gravel}
  n23 = MaterialFunctionCall(
   Material 1=StaticSwitch(True=StaticSwitch(True=@n15.MobileLandscapeMaterial, False=@n18.BlendedMaterial, Value=@n19), False=@n15.MobileLandscapeMaterial, Value=@n20)
   Height Texture 1=StaticSwitch(True=StaticSwitch(True=@n15.Height, False=@n18.Resulting Height Map, Value=@n19), False=@n15.Height, Value=@n20)
   Material 2=@n22.Material
   Height Texture 2=@n22.Material
   Transistion Phase=@n21
  ) {materialFunction=&MF_BlendLandscapeLayers}
  n24 = MaterialFunctionCall!(
   BiomeMask-RGBA=NamedRerouteUsage()
   GravelMaterialInput=@n22.Material
   Heightmap=@n22.Material
   UseUnmodifiedGravelDiffuse=StaticBool() {value=true}
   UnmodifiedGravelDiffuse=@n22.Material
   BypassBiomeColors
  ) {materialFunction=&/MF_Asteria_DirtMudModifier}
  n25 = MaterialFunctionCall(
   Material 1=@n23.BlendedMaterial
   Height Texture 1=@n23.BlendedMaterial
   Material 2=SetMaterialAttributes!(MaterialAttributes=@n22.Material, BaseColor=@n24.ColorizedGravelDiffuse) {attributeSetTypes [#2]}
   Height Texture 2=@n22.Height
   Transistion Phase=LandscapeLayerSample() {parameterName=Mud}
  ) {materialFunction=&MF_BlendLandscapeLayers}
  n26 = MaterialFunctionCall() {materialFunction=&MF_JungleFloor}
  n27 = MaterialFunctionCall(BiomeMask-RGBA=NamedRerouteUsage()) {materialFunction=&MF_Mobile_Asteria_Sand}
  n28 = MaterialFunctionCall(
   Material 1=@n25.BlendedMaterial
   Height Texture 1=@n25.BlendedMaterial
   Material 2=StaticSwitchParameter(True=@n26.MobileMaterial, False=@n27.Material) {parameterName=UseJungleFloor}
   Height Texture 2=StaticSwitchParameter(True=@n26.Height, False=@n27.Height) {parameterName=UseJungleFloor}
   Transistion Phase=LandscapeLayerSample() {parameterName=Sand}
  ) {materialFunction=&MF_BlendLandscapeLayers}
  n29 = NamedRerouteUsage()
  n30 = MaterialFunctionCall!(TilingSize) {materialFunction=&/MF_Mobile_Asteria_Asphault}
  n31 = MaterialFunctionCall(
   Material 1=StaticSwitch(True=@n25.BlendedMaterial, False=@n28.BlendedMaterial, Value=@n29)
   Height Texture 1=StaticSwitch(True=@n25.Resulting Height Map, False=@n28.Resulting Height Map, Value=@n29)
   Material 2=StaticSwitchParameter(True=@n10.MobileMaterial, False=@n30.MobileLandscapeMaterial) {parameterName=UseJungleMud}
   Height Texture 2=StaticSwitchParameter(True=@n10.Height, False=@n30.CrackHeight) {parameterName=UseJungleMud}
   Transistion Phase=LandscapeLayerSample() {parameterName=Asphalt}
  ) {materialFunction=&MF_BlendLandscapeLayers}
  NamedRerouteDeclaration(None=@n31.BlendedMaterial) {name=debugGravel, nodeColor=<rgba 0.7109 1 0 1>}
  NamedRerouteDeclaration(None=FunctionInput(Preview=StaticBool()) {bUsePreviewValueAsDefault=true, inputName=UseSand, inputType=FunctionInput_StaticBool, sortPriority=1}) {name=UseSand, nodeColor=<rgba 1 0.5739 0.0032 1>}
  n32 = MaterialFunctionCall!(BiomeMask-RGBA=NamedRerouteUsage(), UseParameters=StaticBoolParameter() {parameterName=Mobile-Cliffs-UseTextureParameter}) {materialFunction=&/MF_Mobile_Asteria_Cliff}
  NamedRerouteDeclaration(None=@n32.Slope) {name=CliffLayer, nodeColor=<rgba 1 0 0.0703 1>}
  n33 = MaterialFunctionCall(
   Material 1=@n31.BlendedMaterial
   Height Texture 1=@n31.BlendedMaterial
   Material 2=@n32.LandscapeMaterial
   Height Texture 2=@n32.LandscapeMaterial
   Transistion Phase=@n32.LandscapeMaterial
  ) {materialFunction=&MF_BlendLandscapeLayers}
  NamedRerouteDeclaration(None=@n33.BlendedMaterial) {name=CommonLayersMaterial, nodeColor=<rgba 0 1 0.4375 1>}
  n34 = NamedRerouteDeclaration(None=FunctionInput(Preview=Constant()) {bUsePreviewValueAsDefault=true, inputName=BiomeMask-RGBA, inputType=FunctionInput_Vector4}) {name=BiomeMask-RGBA, nodeColor=<rgba 0 0.8988 1 1>}
  n35 = Max(
   A=Max(A=LandscapeLayerSample() {parameterName=Gravel}, B=LandscapeLayerSample() {parameterName=Mud})
   B=LandscapeLayerSample() {parameterName=Sand}
  )
  n36 = LandscapeLayerSample() {parameterName=Asphalt}
  n37 = Max(A=Max(A=Max(A=@n35, B=@n36), B=LandscapeLayerSample() {parameterName=BiomeB}), B=@n9)
  NamedRerouteDeclaration(None=Multiply(A=ComponentMask(None=@n34) {r=true}, B=OneMinus(None=@n37))) {name=SnowLayerMask, nodeColor=<rgba 1 0 0.4453 1>}
  NamedRerouteDeclaration(None=@n33.Resulting Height Map) {name=CommonLayersHeight, nodeColor=<rgba 0.5469 0 1 1>}
  NamedRerouteDeclaration(None=@n32.Mobile-Cliff-SnowMask) {name=Cliff-SnowBuildupMask, nodeColor=<rgba 1 0.2109 0 1>}
  NamedRerouteDeclaration(None=@n24.MudMaterial) {name=Mobile-MudMaterial, nodeColor=<rgba 0 1 0.5781 1>}
  NamedRerouteDeclaration(None=@n17.Material) {name=SandMA, nodeColor=<rgba 1 0.8672 0 1>}
  NamedRerouteUsage()
  NamedRerouteUsage()
  NamedRerouteUsage()
  ScalarParameter() {defaultValue=0.5, parameterName=MFMCL-ShadowAdd}
  n38 = GetMaterialAttributes(None=@n33.BlendedMaterial) {attributeGetTypes [#2]}
  n39 = ComponentMask(None=WorldPosition() {worldPositionShaderOffset=WPT_ExcludeAllShaderOffsets}.XYZ) {b=true}
  n40 = Constant() {r=40}
  n41 = Saturate(None=Divide(A=Add(A=@n39, B=@n40), B=Constant() {r=75}))
  n42 = Constant() {r=0.5}
  n43 = Multiply(A=Saturate(None=Add(A=@n41, B=@n42)), B=@n38.BaseColor)
  n44 = OneMinus(None=SkyAtmosphereLightDirection())
  n45 = ScalarParameter() {defaultValue=4, parameterName=MFMCL-ShadowPow}
  n46 = Saturate(None=DotProduct(A=PixelNormalWS(), B=Power(Base=@n44, Exp=@n45)))
  n47 = Multiply(A=VectorParameter() {parameterName=MFMCL-ShadowColor, defaultValue=<rgba 30 30 30 1>}.RGB, B=@n46)
  SetMaterialAttributes(MaterialAttributes=@n38.MaterialAttributes, BaseColor=LinearInterpolate(A=@n43, B=Multiply(A=@n47, B=@n43), Alpha=@n46)) {attributeSetTypes [#2]}
  SetMaterialAttributes(MaterialAttributes=@n38.MaterialAttributes, BaseColor=@n43) {attributeSetTypes [#2]}
 ]}
