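# Landscape material for the Asteria tundra biome, which also drives grass spawning and physical surfaces such as snow and ice.
# Common layers over a snow base blend with scree, slippery snow, detail snow and cliffs, forming the high-end material for Runtime Virtual Texture writes.
# Shading samples that RVT or uses a simpler mobile material.
# A color-distance snow mask then drives post-RVT detail on higher feature levels, before fire, debug and Lumen adjustments.
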Material /Game/Athena/Asteria/Environments/Landscape/Materials/M_Asteria_Tundra
graph {
 properties {
  bTangentSpaceNormal=false
  bUseMaterialAttributes=true
  bUsedWithNanite=true
  bUsedWithVirtualHeightfieldMesh=true
  pixelDepthOffsetMode=PDOM_Legacy
  shadingModel=MSM_FromMaterialExpression
  usedShadingModels=Default Lit | Subsurface
 }
 nodes [
  n1 = MaterialFunctionCall!(
   LayerMask=Constant() {r=1}
   InputMaterialHeightmap=Constant()
   UseVertexNormalsForBasicSnow
   UseSparkle
   UseTangentSnowNormals
   IsTangentSpaceInput
   Material
  ) {materialFunction=/Game/Athena/Asteria/Environments/Landscape/Materials/MaterialFunctions/MF_Asteria_Snow}
  n2 = MaterialFunctionCall!(
   AdditionalGrassBorderColorMask=LandscapeLayerSample() {desc=ForestFloor, parameterName=BiomeA}
   CustombaseLayerMaterial=@n1.Result
   CustombaseLayerHeight=@n1.Result
   BiomeMask-RGBA=NamedRerouteUsage()
   Custom Biome Mask
   PutSandBeforeGravel=StaticBool()
  ) {materialFunction=/Game/Athena/Asteria/Environments/Landscape/Materials/MaterialFunctions/MF_AsteriaCommonLandscape}
  n3 = MaterialFunctionCall!(BiomeMask-Tundra=NamedRerouteUsage()) {materialFunction=/Game/Athena/Asteria/Environments/Landscape/Materials/MaterialFunctions/MF_Asteria_Scree}
  n4 = MaterialFunctionCall!(
   Material 1=NamedRerouteUsage()
   Height Texture 1=@n2.Heightmap
   Material 2=@n3.Material
   Height Texture 2=@n3.Material
   Transistion Phase=LandscapeLayerSample() {desc=ForestFloor, parameterName=BiomeA}
   UseContrast=StaticBool() {value=true}
   Contrast=ScalarParameter() {defaultValue=0.5, parameterName=ScreeBlendContras}
   AddLayerMask
  ) {materialFunction=/Game/Athena/Apollo/Environments/Landscape/Materials/Biome/CommonMaterials/MF_BlendLandscapeLayers}
  n5 = MaterialFunctionCall() {materialFunction=/Game/Athena/Asteria/Environments/Landscape/Materials/MaterialFunctions/MF_Asteria_SlipperSnow}
  n6 = LandscapeLayerSample() {parameterName=BiomeB}
  n7 = NamedRerouteUsage()
  n8 = MaterialFunctionCall(
   Material 1=@n4.BlendedMaterial
   Height Texture 1=@n4.BlendedMaterial
   Material 2=@n5.Result
   Height Texture 2=@n5.Result
   Transistion Phase=Multiply(A=NamedRerouteDeclaration(None=@n6) {name=SL_SnowMask, nodeColor=<rgba 0 1 0.6953 1>}, B=@n7)
  ) {materialFunction=&MF_BlendLandscapeLayers}
  n9 = MaterialFunctionCall() {materialFunction=/Game/Athena/Asteria/Environments/Landscape/Materials/MaterialFunctions/MF_DetailSnow_01}
  n10 = MaterialFunctionCall(
   Material 1=@n8.BlendedMaterial
   Height Texture 1=@n8.BlendedMaterial
   Material 2=@n9.Material
   Height Texture 2=@n9.Material
   Transistion Phase=LandscapeLayerSample() {parameterName=BiomeC}
  ) {materialFunction=&MF_BlendLandscapeLayers}
  n11 = VertexNormalWS()
  n12 = MaterialFunctionCall!(IceBaseColor, BiomeMask-Tundra=NamedRerouteUsage(), Slope=Abs(None=ComponentMask(None=@n11) {b=true})) {materialFunction=/Game/Athena/Asteria/Environments/Landscape/Materials/MaterialFunctions/MF_Asteria_Tundra_Cliff}
  n13 = OneMinus(None=@n12.Cliff-SnowMask)
  n14 = MaterialFunctionCall!(
   Material=@n10.BlendedMaterial
   Heightmap=@n10.BlendedMaterial
   CliffSnowMask=@n13
   CustomBiomeMask
   NewSlopeMaskMask
   Displacement
   AsphaltLayer=LandscapeLayerSample() {parameterName=Sand}
   AltCliffSlopeScreeMask=@n12.SlopeScreeMask
   UseAlternateCliffMat=StaticBool() {value=true}
   AltCliffHeight=@n12.SlopeScreeMask
   AltCliffMat=@n12.SlopeScreeMask
  ) {materialFunction=/Game/Athena/Asteria/Environments/Landscape/Materials/MaterialFunctions/MF_AsteriaCliffBlend}
  n15 = MaterialFunctionCall!(Material=@n14.OutputMaterial2) {materialFunction=/Game/Athena/Apollo/Maps/Landscape/VT/MaterialFunctions/TransformTangentToWorld}
  n16 = GetMaterialAttributes(None=MaterialFunctionCall!(
    MobileMaterial=NamedRerouteUsage()
    HighEndMaterial=NamedRerouteDeclaration(None=@n15.Result) {name=MaterialForRVT, nodeColor=<rgba 0.5 0 1 1>}
   ) {materialFunction=/Game/Athena/Asteria/Environments/Landscape/Materials/MaterialFunctions/FN_RvtWriteSwitch}.Result) {attributeGetTypes [
    #1
    #2
    #3
    #4
   ]}
  n17 = WorldPosition() {worldPositionShaderOffset=WPT_ExcludeAllShaderOffsets}
  n18 = TextureSample(UVs=Divide(A=ComponentMask(None=@n17.XYZ) {g=true, r=true}, B=ScalarParameter() {defaultValue=80, parameterName=NR-testcompress-tile}) {constB=8192}) {samplerSource=SSM_Wrap_WorldGroupSettings, samplerType=SAMPLERTYPE_LinearColor, texture=/Game/Global/Textures/Default/Noise/T_Apollo_TCNoise}
  Add(A=@n16.Normal, B=Multiply(A=@n18.RGB, B=VectorParameter() {parameterName=NR-testcompress-rgb, defaultValue=<rgba 0.2 0.2 0 1>}.RGB))
  n19 = LandscapeVisibilityMask()
  n20 = RuntimeVirtualTextureSampleParameter() {
   group=RVT
   materialType=BaseColor_Normal_Specular_YCoCg
   parameterName=RVT Sample Target
   virtualTexture=/Game/Athena/Apollo/Maps/Landscape/VT/RVT_Apollo_Landscape
  }
  n21 = SetMaterialAttributes!(MaterialAttributes, Base Color=@n20.BaseColor, Specular=@n20.BaseColor, Roughness=@n20.BaseColor, Normal=@n20.BaseColor) {attributeSetTypes [
    #1
    #2
    #3
    #4
   ]}
  n22 = NamedRerouteUsage()
  n23 = ComponentMask(None=@n22) {r=true}
  n24 = NamedRerouteUsage()
  n25 = MaterialFunctionCall!(PutSandBeforeGravel, BiomeMask-RGBA=@n22, SnowMask=Saturate(None=Subtract(A=@n23, B=@n24)), UseSand, UseSnow=StaticBool() {value=true}) {materialFunction=/Game/Athena/Asteria/Environments/Landscape/Materials/MaterialFunctions/MobileLayers/MF_Asteria_Mobile_CommonLandscapeLayers}
  n26 = GetMaterialAttributes(None=@n25.Material) {attributeGetTypes [#1]}
  MaterialExpressionConstant_13 = Constant()
  n27 = LandscapeLayerBlend!(
   Layer Gravel=@MaterialExpressionConstant_13
   Layer Mud=@MaterialExpressionConstant_13
   Layer Sand=@MaterialExpressionConstant_13
   Layer Asphalt=@MaterialExpressionConstant_13
  ) {layers [
    {
     blendType=LB_WeightBlend
     constHeightInput=0
     layerName=Gravel
     previewWeight=0
     constLayerInput=<0 0 0>
     heightInput {expression=null, inputName=None, mask=0, maskA=0, maskB=0, maskG=0, maskR=0, outputIndex=0}
     layerInput {
      expression=&self:MaterialExpressionConstant_13
      inputName=None
      mask=0
      maskA=0
      maskB=0
      maskG=0
      maskR=0
      outputIndex=0
     }
    }
    {
     blendType=LB_WeightBlend
     constHeightInput=0
     layerName=Mud
     previewWeight=0
     constLayerInput=<0 0 0>
     heightInput {expression=null, inputName=None, mask=0, maskA=0, maskB=0, maskG=0, maskR=0, outputIndex=0}
     layerInput {
      expression=&self:MaterialExpressionConstant_13
      inputName=None
      mask=0
      maskA=0
      maskB=0
      maskG=0
      maskR=0
      outputIndex=0
     }
    }
    {
     blendType=LB_WeightBlend
     constHeightInput=0
     layerName=Sand
     previewWeight=0
     constLayerInput=<0 0 0>
     heightInput {expression=null, inputName=None, mask=0, maskA=0, maskB=0, maskG=0, maskR=0, outputIndex=0}
     layerInput {
      expression=&self:MaterialExpressionConstant_13
      inputName=None
      mask=0
      maskA=0
      maskB=0
      maskG=0
      maskR=0
      outputIndex=0
     }
    }
    {
     blendType=LB_WeightBlend
     constHeightInput=0
     layerName=Asphalt
     previewWeight=0
     constLayerInput=<0 0 0>
     heightInput {expression=null, inputName=None, mask=0, maskA=0, maskB=0, maskG=0, maskR=0, outputIndex=0}
     layerInput {
      expression=&self:MaterialExpressionConstant_13
      inputName=None
      mask=0
      maskA=0
      maskB=0
      maskG=0
      maskR=0
      outputIndex=0
     }
    }
   ]}
  n28 = SetMaterialAttributes!(MaterialAttributes=@n26.MaterialAttributes, Base Color=Add(A=@n26.Base Color, B=Multiply(A=@n27) {constB=0})) {desc="This must remain in material,\r\nnot a function.", attributeSetTypes [#1]}
  n29 = MaterialFunctionCall(NoRVT=@n28, RVT=@n21) {materialFunction=&FN_RVT_Switch_Material}
  n30 = ScalarParameter() {defaultValue=0.812, group=PostRVTSnowMasking, parameterName=PostRVT-SnowMaskTolerance, sliderMax=1}
  n31 = ScalarParameter() {defaultValue=0.25, group=PostRVTSnowMasking, parameterName=PostRVT-SnowMaskToleranceMedium, sliderMax=1}
  n32 = SphereMask(
   A=GetMaterialAttributes(None=@n29.Result) {attributeGetTypes [#1]}.Base Color
   B=MaterialFunctionCall() {materialFunction=/Game/Athena/Asteria/Environments/Landscape/Materials/MaterialFunctions/MF_SnowColorBase}.SnowAlbedo
   Radius=QualitySwitch(Default=@n30, Low=@n31, Medium=@n31, Epic=@n30)
   Hardness=ScalarParameter() {defaultValue=0.5, group=PostRVTSnowMasking, parameterName=PostRVT-SnowMaskHardness, sliderMax=1}
  )
  n33 = FeatureLevelSwitch(
   Default=MaterialFunctionCall!(Height_Input, Material=@n29.Result, SnowMask=@n32, CliffVeinsMask, SL_Snow_Mask=NamedRerouteUsage()) {materialFunction=/Game/Athena/Asteria/Environments/Landscape/Materials/MaterialFunctions/MF_PostRVTDetails}.Result
   ES2=@n29.Result
   ES3_1=@n29.Result
  )
  n34 = MaterialFunctionCall!(
   Material=MaterialFunctionCall!(
    In=@n33
    UseHeightFalloff
    OnFireEmissiveOverride
    HeightFalloffRadius
    HeightFalloffHardness
    VariationMasksMobile
    In_Grass
    Is_Stylised_WE
   ) {desc=Fire, materialFunction=/Game/Packages/Fortress_SharedMaterials/Base_Material_Functions/MF_Curie_Landscape}.Result
   DebugInputAsBasecolor
   Non-RVT material=NamedRerouteUsage()
   DebugMaterialAs1Color=VectorParameter() {parameterName=DebugBiomeColor, defaultValue=<rgba 1 0.0502 0.0203 1>}.RGB
  ) {materialFunction=/Game/Athena/Apollo/Environments/Landscape/Materials/Biome/CommonMaterials/MF_LS_DEBUG}
  n35 = MaterialFunctionCall(Y=@n32) {materialFunction=&MakeFloat3}
  BlendMaterialAttributes(
   A=MaterialFunctionCall!(In=SetMaterialAttributes!(MaterialAttributes=@n34.Result, Opacity Mask=@n19) {attributeSetTypes [#5]}) {materialFunction=/Game/Athena/Asteria/Environments/Landscape/Materials/MaterialFunctions/MF_Terrain_LumenModification}.Result
   B=MakeMaterialAttributes!(
    BaseColor=@n35.Result
    Metallic
    Specular
    Roughness
    Anisotropy
    EmissiveColor=@n35.Result
    Opacity
    OpacityMask
    Normal
    Tangent
    WorldPositionOffset
    SubsurfaceColor
    ClearCoat
    ClearCoatRoughness
    AmbientOcclusion
    Refraction
    CustomizedUVs_0
    CustomizedUVs_1
    CustomizedUVs_2
    CustomizedUVs_3
    CustomizedUVs_4
    CustomizedUVs_5
    CustomizedUVs_6
    CustomizedUVs_7
    PixelDepthOffset
    ShadingModel
    Displacement
   )
   Alpha=MaterialFunctionCall() {materialFunction=/Engine/Functions/Engine_MaterialFunctions02/Utility/DebugOnOff}.Result
  )
  Constant3Vector() {constant=<rgba 0.03 0.03 0 1>}
  Constant() {r=50}
  n36 = Reroute()
  n37 = Reroute()
  n38 = Constant3Vector() {constant=<rgba 0 1 0 0>}
  n39 = If(A=Constant(), B=@n37, "A > B"=@n38, "A == B"=@n38, "A < B"=Constant3Vector() {constant=<rgba 1 0 0 0>})
  n40 = If(A=@n37, B=@n36, "A > B"=@n39, "A == B"=@n39, "A < B"=Constant3Vector() {constant=<rgba 0 0 1 0>})
  If(A=@n36, B=Reroute(), "A > B"=@n40, "A == B"=@n40, "A < B"=@n38)
  MaterialExpressionMaterialFunctionCall_9 = MaterialFunctionCall!(In=NamedRerouteUsage()) {materialFunction=/Game/Athena/Apollo/Environments/Landscape/Materials/Biome/CommonMaterials/MF_LS_GrassSwitch}
  MaterialExpressionMaterialFunctionCall_14 = MaterialFunctionCall(In=NamedRerouteUsage()) {materialFunction=&MF_LS_GrassSwitch}
  MaterialExpressionMaterialFunctionCall_15 = MaterialFunctionCall(In=NamedRerouteUsage()) {materialFunction=&MF_LS_GrassSwitch}
  MaterialExpressionMaterialFunctionCall_19 = MaterialFunctionCall(In=NamedRerouteUsage()) {materialFunction=&MF_LS_GrassSwitch}
  MaterialExpressionMaterialFunctionCall_35 = MaterialFunctionCall(In=Multiply(A=NamedRerouteUsage(), B=NamedRerouteUsage())) {materialFunction=&MF_LS_GrassSwitch}
  MaterialExpressionMaterialFunctionCall_36 = MaterialFunctionCall(In=Multiply(A=NamedRerouteUsage(), B=NamedRerouteUsage())) {materialFunction=&MF_LS_GrassSwitch}
  n41 = Constant()
  MaterialExpressionMaterialFunctionCall_6 = MaterialFunctionCall(In=@n41) {materialFunction=&MF_LS_GrassSwitch}
  MaterialExpressionMaterialFunctionCall_8 = MaterialFunctionCall(In=NamedRerouteUsage()) {materialFunction=&MF_LS_GrassSwitch}
  MaterialExpressionMaterialFunctionCall_3 = MaterialFunctionCall(In=NamedRerouteUsage()) {materialFunction=&MF_LS_GrassSwitch}
  MaterialExpressionMaterialFunctionCall_18 = MaterialFunctionCall(In=@n41) {materialFunction=&MF_LS_GrassSwitch}
  MaterialExpressionMaterialFunctionCall_40 = MaterialFunctionCall(In=NamedRerouteUsage()) {materialFunction=&MF_LS_GrassSwitch}
  MaterialExpressionMaterialFunctionCall_44 = MaterialFunctionCall() {materialFunction=&MF_LS_GrassSwitch}
  MaterialExpressionMaterialFunctionCall_47 = MaterialFunctionCall() {materialFunction=&MF_LS_GrassSwitch}
  MaterialExpressionMaterialFunctionCall_48 = MaterialFunctionCall() {materialFunction=&MF_LS_GrassSwitch}
  MaterialExpressionMaterialFunctionCall_50 = MaterialFunctionCall() {materialFunction=&MF_LS_GrassSwitch}
  MaterialExpressionMaterialFunctionCall_55 = MaterialFunctionCall(In=NamedRerouteUsage()) {materialFunction=&MF_LS_GrassSwitch}
  LandscapeGrassOutput!(
   Grass_Common_Layer1=@MaterialExpressionMaterialFunctionCall_9.Result
   Grass_Common_Layer2=@MaterialExpressionMaterialFunctionCall_14.Result
   Grass_Common_Layer3=@MaterialExpressionMaterialFunctionCall_15.Result
   GravelScree=@MaterialExpressionMaterialFunctionCall_19.Result
   GravelScree_02=@MaterialExpressionMaterialFunctionCall_35.Result
   GravelScree_03=@MaterialExpressionMaterialFunctionCall_36.Result
   Common_ForestFloor=@MaterialExpressionMaterialFunctionCall_6.Result
   Tundra_SnowScree=@MaterialExpressionMaterialFunctionCall_8.Result
   Common_Scree=@MaterialExpressionMaterialFunctionCall_3.Result
   Fall_ForestFloor=@MaterialExpressionMaterialFunctionCall_18.Result
   Grass_Fall_Layer1=@MaterialExpressionMaterialFunctionCall_40.Result
   Common_FlowerField=@MaterialExpressionMaterialFunctionCall_6.Result
   Fall_FlowerField=@MaterialExpressionMaterialFunctionCall_18.Result
   Beach_Scree=@MaterialExpressionMaterialFunctionCall_44.Result
   Beach_Foliage=@MaterialExpressionMaterialFunctionCall_47.Result
   Beach_Foliage_02=@MaterialExpressionMaterialFunctionCall_48.Result
   Tundra_Foliage=@MaterialExpressionMaterialFunctionCall_50.Result
   JungleFloor_01=@MaterialExpressionMaterialFunctionCall_50.Result
   JungleFloor_02=@MaterialExpressionMaterialFunctionCall_50.Result
   JungleFloor_03=@MaterialExpressionMaterialFunctionCall_50.Result
   Bush=@MaterialExpressionMaterialFunctionCall_55.Result
   Jungle_Moss=@MaterialExpressionMaterialFunctionCall_50.Result
  ) {grassTypes [
    Grass_Common_Layer1 {
     grassType=/Game/Environments/Asteria/Foliage/Grass/Spawners/Groda_LS_Grass_Standard
     input {
      expression=&self:MaterialExpressionMaterialFunctionCall_9
      inputName=None
      mask=0
      maskA=0
      maskB=0
      maskG=0
      maskR=0
      outputIndex=0
     }
    }
    Grass_Common_Layer2 {
     grassType=/Game/Environments/Asteria/Foliage/Grass/Spawners/LS_Grass_Standard_Variation_01
     input {
      expression=&self:MaterialExpressionMaterialFunctionCall_14
      inputName=None
      mask=0
      maskA=0
      maskB=0
      maskG=0
      maskR=0
      outputIndex=0
     }
    }
    Grass_Common_Layer3 {
     grassType=/Game/Environments/Asteria/Foliage/Grass/Spawners/LS_Grass_Standard_Variation_02
     input {
      expression=&self:MaterialExpressionMaterialFunctionCall_15
      inputName=None
      mask=0
      maskA=0
      maskB=0
      maskG=0
      maskR=0
      outputIndex=0
     }
    }
    GravelScree {
     grassType=/Game/Environments/Asteria/Foliage/Grass/Spawners/LS_Gravel_Scree_V1
     input {
      expression=&self:MaterialExpressionMaterialFunctionCall_19
      inputName=None
      mask=0
      maskA=0
      maskB=0
      maskG=0
      maskR=0
      outputIndex=0
     }
    }
    GravelScree_02 {
     grassType=/Game/Environments/Asteria/Foliage/Grass/Spawners/LS_Gravel_Scree_V2
     input {
      expression=&self:MaterialExpressionMaterialFunctionCall_35
      inputName=None
      mask=0
      maskA=0
      maskB=0
      maskG=0
      maskR=0
      outputIndex=0
     }
    }
    GravelScree_03 {
     grassType=/Game/Environments/Asteria/Foliage/Grass/Spawners/LS_Gravel_Scree_V3
     input {
      expression=&self:MaterialExpressionMaterialFunctionCall_36
      inputName=None
      mask=0
      maskA=0
      maskB=0
      maskG=0
      maskR=0
      outputIndex=0
     }
    }
    Common_ForestFloor {
     grassType=/Game/Environments/Asteria/Foliage/Grass/Spawners/LS_Common_ForestFloor
     input {
      expression=&self:MaterialExpressionMaterialFunctionCall_6
      inputName=None
      mask=0
      maskA=0
      maskB=0
      maskG=0
      maskR=0
      outputIndex=0
     }
    }
    Tundra_SnowScree {
     grassType=/Game/Environments/Asteria/Foliage/Grass/Spawners/LS_Snow_Standard
     input {
      expression=&self:MaterialExpressionMaterialFunctionCall_8
      inputName=None
      mask=0
      maskA=0
      maskB=0
      maskG=0
      maskR=0
      outputIndex=0
     }
    }
    Common_Scree {
     grassType=/Game/Environments/Asteria/Foliage/Grass/Spawners/LS_RubbleScree
     input {
      expression=&self:MaterialExpressionMaterialFunctionCall_3
      inputName=None
      mask=0
      maskA=0
      maskB=0
      maskG=0
      maskR=0
      outputIndex=0
     }
    }
    Fall_ForestFloor {
     grassType=/Game/Environments/Asteria/Foliage/Grass/Spawners/LS_ForestFloorScree
     input {
      expression=&self:MaterialExpressionMaterialFunctionCall_18
      inputName=None
      mask=0
      maskA=0
      maskB=0
      maskG=0
      maskR=0
      outputIndex=0
     }
    }
    Grass_Fall_Layer1 {
     grassType=/Game/Environments/Asteria/Foliage/Grass/Spawners/LS_Grass_Fall
     input {
      expression=&self:MaterialExpressionMaterialFunctionCall_40
      inputName=None
      mask=0
      maskA=0
      maskB=0
      maskG=0
      maskR=0
      outputIndex=0
     }
    }
    Common_FlowerField {
     grassType=/Game/Environments/Asteria/Foliage/Grass/Spawners/LS_Common_FlowerField
     input {
      expression=&self:MaterialExpressionMaterialFunctionCall_6
      inputName=None
      mask=0
      maskA=0
      maskB=0
      maskG=0
      maskR=0
      outputIndex=0
     }
    }
    Fall_FlowerField {
     grassType=/Game/Environments/Asteria/Foliage/Grass/Spawners/LS_Fall_FlowerField
     input {
      expression=&self:MaterialExpressionMaterialFunctionCall_18
      inputName=None
      mask=0
      maskA=0
      maskB=0
      maskG=0
      maskR=0
      outputIndex=0
     }
    }
    Beach_Scree {
     grassType=/Game/Environments/Asteria/Foliage/Grass/Spawners/LS_BeachScree
     input {
      expression=&self:MaterialExpressionMaterialFunctionCall_44
      inputName=None
      mask=0
      maskA=0
      maskB=0
      maskG=0
      maskR=0
      outputIndex=0
     }
    }
    Beach_Foliage {
     grassType=/Game/Environments/Asteria/Foliage/Grass/Spawners/LS_BeachFoliage
     input {
      expression=&self:MaterialExpressionMaterialFunctionCall_47
      inputName=None
      mask=0
      maskA=0
      maskB=0
      maskG=0
      maskR=0
      outputIndex=0
     }
    }
    Beach_Foliage_02 {
     grassType=/Game/Environments/Asteria/Foliage/Grass/Spawners/LS_BeachFoliage_02
     input {
      expression=&self:MaterialExpressionMaterialFunctionCall_48
      inputName=None
      mask=0
      maskA=0
      maskB=0
      maskG=0
      maskR=0
      outputIndex=0
     }
    }
    Tundra_Foliage {
     grassType=/Game/Environments/Asteria/Foliage/Grass/Spawners/LS_TundraFoliage
     input {
      expression=&self:MaterialExpressionMaterialFunctionCall_50
      inputName=None
      mask=0
      maskA=0
      maskB=0
      maskG=0
      maskR=0
      outputIndex=0
     }
    }
    JungleFloor_01 {
     grassType=/Game/Environments/Asteria/Foliage/Grass/Spawners/LS_JungleFloor_01
     input {
      expression=&self:MaterialExpressionMaterialFunctionCall_50
      inputName=None
      mask=0
      maskA=0
      maskB=0
      maskG=0
      maskR=0
      outputIndex=0
     }
    }
    JungleFloor_02 {
     grassType=/Game/Environments/Asteria/Foliage/Grass/Spawners/LS_JungleFloor_02
     input {
      expression=&self:MaterialExpressionMaterialFunctionCall_50
      inputName=None
      mask=0
      maskA=0
      maskB=0
      maskG=0
      maskR=0
      outputIndex=0
     }
    }
    JungleFloor_03 {
     grassType=/Game/Environments/Asteria/Foliage/Grass/Spawners/LS_JungleFloor_03
     input {
      expression=&self:MaterialExpressionMaterialFunctionCall_50
      inputName=None
      mask=0
      maskA=0
      maskB=0
      maskG=0
      maskR=0
      outputIndex=0
     }
    }
    Bush {
     grassType=/Game/Environments/Asteria/Foliage/Grass/Spawners/Groda_LS_Grass_Bush
     input {
      expression=&self:MaterialExpressionMaterialFunctionCall_55
      inputName=None
      mask=0
      maskA=0
      maskB=0
      maskG=0
      maskR=0
      outputIndex=0
     }
    }
    Jungle_Moss {
     grassType=/Game/Environments/Asteria/Foliage/Grass/Spawners/LS_JungleMoss
     input {
      expression=&self:MaterialExpressionMaterialFunctionCall_50
      inputName=None
      mask=0
      maskA=0
      maskB=0
      maskG=0
      maskR=0
      outputIndex=0
     }
    }
   ]}
  LandscapeLayerSample() {parameterName=Gravel}
  n42 = NamedRerouteUsage()
  n43 = OneMinus(None=@n42) {desc="White=Cliffs"}
  n44 = LandscapeLayerSample() {parameterName=Gravel}
  n45 = LandscapeLayerSample() {parameterName=Mud}
  n46 = Saturate(None=Multiply(A=Max(A=@n44, B=@n45), B=Constant() {r=2}))
  n47 = Subtract(A=@n46, B=@n43) {desc="Remove Cliff from Gravel,Mud,Sand"}
  n48 = LandscapeLayerSample() {parameterName=Sand}
  n49 = Saturate(None=Multiply(A=LandscapeLayerSample() {parameterName=Asphalt}, B=Constant() {r=3}))
  n50 = Max(A=Max(A=NamedRerouteUsage(), B=NamedRerouteUsage()), B=@n49)
  MaterialExpressionSaturate_12 = Saturate(None=Subtract(A=Saturate(None=@n47), B=Max(A=@n48, B=@n50)) {desc="Remove ForestFloor, Flowers, Asphalt, Sand"}) {desc=Final Dirt mask}
  MaterialExpressionReroute_25 = Reroute(None=@n43) {desc=Final Stone mask}
  n51 = Saturate(None=Subtract(A=@n42, B=Max(A=@n48, B=@n46)) {desc="Remove Gravel,Mud,Sand"})
  MaterialExpressionMultiply_12 = Multiply(A=NamedRerouteUsage(), B=NamedRerouteUsage())
  n52 = Subtract(A=Saturate(None=Subtract(A=@n51, B=@n49) {desc=Remove Asphalt}), B=@MaterialExpressionMultiply_12)
  MaterialExpressionClamp_0 = Clamp(None=NamedRerouteUsage())
  MaterialExpressionSaturate_14 = Saturate(None=Subtract(A=Saturate(None=@n52), B=@MaterialExpressionClamp_0)) {desc=Final Grass mask}
  MaterialExpressionReroute_29 = Reroute(None=@n49)
  n53 = Subtract(A=@n48, B=@n43) {desc=Remove Cliff}
  n54 = Saturate(None=Subtract(A=Saturate(None=@n53), B=@n50) {desc="Remove ForestFloor, Flowers, Asphalt"})
  MaterialExpressionReroute_16 = Reroute(None=@n54)
  LandscapePhysicalMaterialOutput!(
   Dirt=@MaterialExpressionSaturate_12
   Stone=@MaterialExpressionReroute_25
   Grass=@MaterialExpressionSaturate_14
   Asphalt=@MaterialExpressionReroute_29
   Sand=@MaterialExpressionReroute_16
   PhysData_Ice=@MaterialExpressionMultiply_12
   Snow=@MaterialExpressionClamp_0
  ) {
   desc="These are layered in order of importance.\r\nIf Dirt=1 and Stone=1, then Dirt will \"win\".\r\nTheoretically, these values aren't clamped.\r\nSo a value of 2 should trump a value of 1."
   inputs [
    {
     physicalMaterial=/Game/FortressPhysicalMaterials/Dirt
     input {
      expression=&self:MaterialExpressionSaturate_12
      inputName=None
      mask=0
      maskA=0
      maskB=0
      maskG=0
      maskR=0
      outputIndex=0
     }
    }
    {
     physicalMaterial=/Game/FortressPhysicalMaterials/Stone
     input {
      expression=&self:MaterialExpressionReroute_25
      inputName=None
      mask=0
      maskA=0
      maskB=0
      maskG=0
      maskR=0
      outputIndex=0
     }
    }
    {
     physicalMaterial=/Game/FortressPhysicalMaterials/Grass
     input {
      expression=&self:MaterialExpressionSaturate_14
      inputName=None
      mask=0
      maskA=0
      maskB=0
      maskG=0
      maskR=0
      outputIndex=0
     }
    }
    {
     physicalMaterial=/Game/FortressPhysicalMaterials/Asphalt
     input {
      expression=&self:MaterialExpressionReroute_29
      inputName=None
      mask=0
      maskA=0
      maskB=0
      maskG=0
      maskR=0
      outputIndex=0
     }
    }
    {
     physicalMaterial=/Game/FortressPhysicalMaterials/Sand
     input {
      expression=&self:MaterialExpressionReroute_16
      inputName=None
      mask=0
      maskA=0
      maskB=0
      maskG=0
      maskR=0
      outputIndex=0
     }
    }
    {
     physicalMaterial=/Game/Balance/Physics/Materials/PhysData_Ice
     input {
      expression=&self:MaterialExpressionMultiply_12
      inputName=None
      mask=0
      maskA=0
      maskB=0
      maskG=0
      maskR=0
      outputIndex=0
     }
    }
    {
     physicalMaterial=/Game/FortressPhysicalMaterials/Snow
     input {
      expression=&self:MaterialExpressionClamp_0
      inputName=None
      mask=0
      maskA=0
      maskB=0
      maskG=0
      maskR=0
      outputIndex=0
     }
    }
   ]
  }
  n55 = LinearInterpolate(A=@MaterialExpressionSaturate_12, B=@MaterialExpressionReroute_25, Alpha=ScalarParameter() {parameterName=DEBUG-PhysMat1})
  n56 = ScalarParameter() {parameterName=DEBUG-PhysMat2}
  LinearInterpolate(
   A=LinearInterpolate(
    A=LinearInterpolate(A=@n55, B=@MaterialExpressionSaturate_14, Alpha=@n56)
    B=@MaterialExpressionReroute_29
    Alpha=ScalarParameter() {parameterName=DEBUG-PhysMat3}
   )
   B=@n54
   Alpha=ScalarParameter() {defaultValue=1, parameterName=DEBUG-PhysMat4}
  )
  LinearInterpolate(Alpha=MaterialFunctionCall() {materialFunction=&DebugTimeSine}.Result)
  n57 = StaticBool()
  MaterialFunctionCall!(
   IsStaticSnow
   UseAlternateStaticSnowMask
   AlternateStaticSnowMask
   "UseRoad&&OcclusionMask"
   IsForGrassMeshToPushDown
   AdditionalMaskToRemoveSnow
   IsFallingSnowFX
   UseMinimapCoords
   IncreaseSnowUsingSlope
   UseSmoothBorder
   LerpHighM
   TemporalHeightLerpContrast
   IsLobbyIsland
   UseBR-IslandMask=@n57
   IsTree
   IsForRenderTarget
   UseSparkle
   UseSnowNormals=StaticBool() {value=true}
   UseAdditionalMaskToRemoveSnow
   IsTangentSpaceInput
   UseVertexNormalsForBasicSnow
   UseTangentSnowNormals=@n57
   Material=@n33
  ) {materialFunction=/Game/Athena/Apollo/Environments/Landscape/Materials/Biome/UniqueBiomeMaterialFunctions/MF_Apollo_Snow_Progression}
  n58 = StaticSwitch(True=LandscapeLayerSample() {parameterName=Sand}, False=LandscapeLayerSample() {parameterName=Mud}, Value=NamedRerouteUsage())
  n59 = NamedRerouteUsage()
  n60 = OneMinus(None=NamedRerouteUsage())
  Max(A=Max(A=Max(A=@n58, B=@n59), B=LandscapeLayerSample() {parameterName=Asphalt}), B=@n60)
  Multiply(A=Max(A=NamedRerouteUsage(), B=NamedRerouteUsage()), B=NamedRerouteUsage())
  n61 = ScalarParameter() {defaultValue=-0.5, parameterName=BiomeC-FFSpawnLa}
  n62 = ScalarParameter() {defaultValue=1.5, parameterName=BiomeC-FFSpawnLb}
  n63 = LandscapeLayerSample() {desc=ForestFloor, parameterName=BiomeC}
  n64 = MaterialFunctionCall(
   Transition Phase=Saturate(None=LinearInterpolate(A=@n61, B=@n62, Alpha=@n63))
   Height Texture=NamedRerouteUsage()
   Contrast=ScalarParameter() {defaultValue=0.5, parameterName=BiomeC-FFSpawnCont}
  ) {desc=ForestFloor, materialFunction=&HeightLerp}
  Multiply(A=OneMinus(None=NamedRerouteUsage()), B=@n64.Alpha)
  n65 = TextureObject() {texture=/Game/Environments/Asteria/Terrain/Textures/T_Asteria_WindBlownSnow_01_D}
  n66 = MaterialFunctionCall!(Scale=ScalarParameter() {defaultValue=1800, parameterName=NoiseScale1}) {materialFunction=/Game/Athena/Apollo/Environments/Landscape/Materials/Biome/CommonMaterials/MF_SimpleWorldUV}
  n67 = Multiply(A=@n66.Result) {constB=0.5}
  n68 = ScalarParameter() {defaultValue=1, parameterName=HeightScale}
  n69 = Constant()
  n70 = Constant()
  Multiply(
   A=ComponentMask(None=MaterialFunctionCall!(
     Texture=@n65
     UseMipBias
     MipBias
     LowEndTiling
     LandscapeCoordinates=@n66.Result
     HexGridSize
     HexGridUVInput=@n67
     HexGridGradientSharpness
     MaxScale=@n68
     MinScale=@n68
     MaxOffset=@n69
     MinOffset=@n69
     "MaxRotation(Degrees)"=@n70
     "MinRotation(Degrees)"=@n70
     forceLowQuality
    ) {materialFunction=/Game/Athena/Asteria/Environments/Landscape/Materials/MaterialFunctions/UtilityFunctions/Landscape_NormalTextureSample}.Result) {a=true}
   B=ScalarParameter() {defaultValue=0.5, parameterName=SnowHeightMapBias}
  )
  NamedRerouteDeclaration(None=@n2.GrassNoiseMask) {name=GrassNoiseMask, nodeColor=<rgba 0.5 0 1 1>}
  NamedRerouteDeclaration(None=@n2.GrassBaseColorDetail) {name=GrassBaseColorDetail, nodeColor=<rgba 0.5 0 1 1>}
  NamedRerouteDeclaration(None=@n2.Material) {name=CommonMaterial, nodeColor=<rgba 0.5 0 1 1>}
  NamedRerouteDeclaration(None=@n9.Height) {name=BiomeC-Height, nodeColor=<rgba 1 0 0.0703 1>}
  NamedRerouteDeclaration(None=OneMinus(None=@n2.SnowScreeMask)) {name=SnowLayerMask, nodeColor=<rgba 1 1 1 1>}
  NamedRerouteDeclaration(None=@n14.SlopeScreeMask) {name=SlopeScreeMask, nodeColor=<rgba 0.5 0 1 1>}
  n71 = Max(A=LandscapeLayerSample() {parameterName=BiomeC}, B=Max(A=NamedRerouteUsage(), B=NamedRerouteUsage()))
  n72 = NamedRerouteUsage()
  n73 = Max(A=Max(A=Max(A=@n71, B=@n72), B=NamedRerouteUsage()), B=NamedRerouteUsage())
  n74 = NamedRerouteUsage()
  n75 = NamedRerouteUsage()
  n76 = NamedRerouteUsage()
  n77 = OneMinus(None=Max(A=Max(A=@n73, B=@n74), B=Max(A=@n75, B=@n76)))
  NamedRerouteDeclaration(None=@n77) {name=GM-Grass, nodeColor=<rgba 0.5 0 1 1>}
  NamedRerouteDeclaration(None=@n77) {name=GM-Bush, nodeColor=<rgba 0.5 0 1 1>}
  n78 = LandscapeLayerSample() {parameterName=Sand}
  n79 = Max(A=NamedRerouteUsage(), B=NamedRerouteUsage())
  n80 = ComponentMask(None=WorldPosition().XYZ) {b=true}
  n81 = Constant() {r=200}
  n82 = Subtract(A=Saturate(None=Subtract(A=@n78, B=@n79)), B=Saturate(None=Divide(A=@n80, B=@n81)))
  n83 = ComponentMask(None=WorldPosition().XYZ) {b=true}
  n84 = Constant() {r=-1200}
  n85 = OneMinus(None=Saturate(None=Subtract(A=@n83, B=@n84)))
  NamedRerouteDeclaration(None=Subtract(A=Saturate(None=@n82), B=@n85)) {name=GM-Beach, nodeColor=<rgba 0.5 0 1 1>}
  n86 = Saturate(None=Subtract(A=@n83, B=Constant() {r=-700}))
  NamedRerouteDeclaration(None=Subtract(A=OneMinus(None=@n86), B=@n85)) {name=GM-Ocean, nodeColor=<rgba 0.5 0 1 1>}
  NamedRerouteDeclaration(None=@n2.CommonLayerMaskForGrassFoliage) {name=GM-CommonLayers, nodeColor=<rgba 0.5 0 1 1>}
  NamedRerouteDeclaration(None=@n2.GrassHeightmap) {name=GrassHeightmap, nodeColor=<rgba 0.5 0 1 1>}
  NamedRerouteDeclaration(None=@n60) {name=GM-FrozenLayers, nodeColor=<rgba 0.0547 0 1 1>}
  n87 = MaterialFunctionCall!(UseLandscapeCoords) {materialFunction=/Game/Athena/Asteria/Environments/Landscape/Materials/MaterialFunctions/MF_AsteriaTerrainMasks}
  NamedRerouteDeclaration(None=@n87.BiomeMaskRGBA) {name=BiomeMask-RGBA, nodeColor=<rgba 0.4531 0 1 1>}
  NamedRerouteDeclaration(None=@n14.HeightmapForSnow) {name=Cliff_HeightMapForSnow, nodeColor=<rgba 1 0.0469 0 1>}
  n88 = MaterialFunctionCall!(
   Texture=TextureObject() {samplerType=SAMPLERTYPE_Grayscale, texture=/Game/Weapons/FORT_Melee/Pickaxe_Cube_Ninja_Male/FX/T_CellNoise}
   LandscapeCoordinates=Multiply(A=ScalarParameter() {defaultValue=0.05, parameterName=GM-GrassVariationMask_01_Tiling}, B=LandscapeLayerCoords())
   LowEndTiling
   HexGridGradientSharpness
   HexGridSize
   HexGridUVInput
   MinScale
   MaxScale
   MinOffset
   MaxOffset
   "MaxRotation(Degrees)"
   "MinRotation(Degrees)"
   forceLowQuality
  ) {materialFunction=/Game/Athena/Asteria/Environments/Landscape/Materials/MaterialFunctions/UtilityFunctions/Landscape_TextureSample}
  n89 = Power(
   Base=Multiply(A=ComponentMask(None=@n88.Result) {r=true}, B=ScalarParameter() {defaultValue=10, parameterName=MaskStrength})
   Exp=ScalarParameter() {defaultValue=4, parameterName=GM-VariationMaskPow}
  )
  NamedRerouteDeclaration(None=Multiply(A=Clamp(None=@n89), B=NamedRerouteUsage())) {name=GM-VariationMask-1, nodeColor=<rgba 1 0 0.8203 1>}
  n90 = MaterialFunctionCall(
   Texture=TextureObject() {samplerType=SAMPLERTYPE_Grayscale, texture=&T_CellNoise}
   LandscapeCoordinates=Multiply(A=ScalarParameter() {defaultValue=0.025, parameterName=GM-GrassVariationMask_02_Tiling}, B=LandscapeLayerCoords())
  ) {materialFunction=&Landscape_TextureSample}
  n91 = Power(
   Base=Multiply(A=ComponentMask(None=@n90.Result) {r=true}, B=ScalarParameter() {defaultValue=2, parameterName=GM_Mask_v2_Strength})
   Exp=ScalarParameter() {defaultValue=7, parameterName=GM-VariationMask_02_Pow}
  )
  NamedRerouteDeclaration(None=Multiply(A=Clamp(None=@n91), B=NamedRerouteUsage())) {name=GM-VariationMask-2, nodeColor=<rgba 0 1 0.8125 1>}
  NamedRerouteDeclaration(None=OneMinus(None=@n73)) {name=GM-CombinedCommonLayers, nodeColor=<rgba 0.6172 1 0 1>}
  NamedRerouteDeclaration(None=@n2.GrassColorMask1) {name=GrassColorMask1, nodeColor=<rgba 0 0.8125 1 1>}
  NamedRerouteDeclaration(None=@n2.GrassColorMask3) {name=GrassColorMask3, nodeColor=<rgba 0 0.8125 1 1>}
  NamedRerouteDeclaration(None=@n2.GrassColorMask2) {name=GrassColorMask2, nodeColor=<rgba 0 0.8125 1 1>}
  NamedRerouteDeclaration(None=@n13) {name=CliffSnowMask, nodeColor=<rgba 1 0 0.75 1>}
  n92 = NamedRerouteUsage()
  n93 = NamedRerouteUsage()
  n94 = Max(A=Max(A=Max(A=@n92, B=@n93), B=NamedRerouteUsage()), B=NamedRerouteUsage())
  NamedRerouteDeclaration(None=Multiply(A=NamedRerouteUsage(), B=OneMinus(None=@n94))) {name=GM-Gravel, nodeColor=<rgba 0.5 0 1 1>}
  n95 = Max(A=NamedRerouteUsage(), B=Max(A=NamedRerouteUsage(), B=NamedRerouteUsage()))
  NamedRerouteDeclaration(None=Multiply(A=NamedRerouteUsage(), B=OneMinus(None=@n95))) {name=GM-Snow, nodeColor=<rgba 1 0 0.6797 1>}
  NamedRerouteDeclaration(None=OneMinus(None=@n8.Alpha)) {name=GM-IceMask, nodeColor=<rgba 0.0781 0 1 1>}
  NamedRerouteDeclaration(None=@MaterialExpressionSaturate_14) {name=GrassPhysMatMask, nodeColor=<rgba 1 0.9609 0 1>}
  NamedRerouteDeclaration(None=@n1.SnowEmissive) {name=Snow-Emissive, nodeColor=<rgba 0 0.7422 1 1>}
  NamedRerouteDeclaration(None=@n2.GravelMaskForGrassFoliage) {name=GravelMask, nodeColor=<rgba 0.5 0 1 1>}
  NamedRerouteDeclaration(None=OneMinus(None=@n4.Alpha)) {name=ScreeMask, nodeColor=<rgba 1 0 0.6797 1>}
  NamedRerouteDeclaration(None=@n2.GM-Mud) {name=MudMask, nodeColor=<rgba 0 1 0.6016 1>}
  n96 = OneMinus(None=NamedRerouteUsage())
  n97 = ScalarParameter() {defaultValue=-0.5, parameterName=BiomeA-FFSpawnLa}
  n98 = ScalarParameter() {defaultValue=1.5, parameterName=BiomeA-FFSpawnLb}
  n99 = LandscapeLayerSample() {desc=ForestFloor, parameterName=BiomeA}
  n100 = MaterialFunctionCall(
   Transition Phase=Saturate(None=LinearInterpolate(A=@n97, B=@n98, Alpha=@n99))
   Height Texture=NamedRerouteUsage()
   Contrast=ScalarParameter() {defaultValue=0.5, parameterName=BiomeA-FFSpawnCont}
  ) {desc=ForestFloor, materialFunction=&HeightLerp}
  NamedRerouteDeclaration(None=Max(A=Multiply(A=@n96, B=@n100.Alpha), B=NamedRerouteUsage())) {name=GM-RubbleScree, nodeColor=<rgba 1 0 0.3281 1>}
  n101 = OneMinus(None=@n74)
  NamedRerouteDeclaration(None=Multiply(A=@n101, B=@n76)) {name=GM-Grass-03, nodeColor=<rgba 1 0 0.7266 1>}
  NamedRerouteDeclaration(None=Multiply(A=Multiply(A=@n101, B=@n75), B=OneMinus(None=@n76))) {name=GM-Grass-02, nodeColor=<rgba 1 0.375 0 1>}
  NamedRerouteDeclaration(None=@n3.Height) {name=BiomeA-Scree-Height, nodeColor=<rgba 0.6928 0.156 1 1>}
  NamedRerouteDeclaration(None=@n2.GM-Sand) {name=SandMask, nodeColor=<rgba 0.5 0 1 1>}
  NamedRerouteDeclaration(None=@n2.GM-Asphalt) {name=AsphaultMask, nodeColor=<rgba 0.5 0 1 1>}
  n102 = ScalarParameter() {defaultValue=0.2, parameterName=GM-GravelVariationMask_01_Tiling}
  n103 = LandscapeLayerCoords()
  n104 = ComponentMask(None=MaterialFunctionCall(
    Texture=TextureObject() {samplerType=SAMPLERTYPE_Grayscale, texture=&T_CellNoise}
    LandscapeCoordinates=Multiply(A=@n102, B=@n103)
   ) {materialFunction=&Landscape_TextureSample}.Result) {r=true}
  n105 = ScalarParameter() {defaultValue=10, parameterName=GravelMaskStrength}
  n106 = Clamp(None=Power(Base=Multiply(A=@n104, B=@n105), Exp=ScalarParameter() {defaultValue=4, parameterName=GM-GravelVariationMaskPow}))
  NamedRerouteDeclaration(None=@n106) {name=VariationMask_01, nodeColor=<rgba 0 1 0.5781 1>}
  n107 = ScalarParameter() {defaultValue=0.075, parameterName=GM-GravelVariationMask_02_Tiling}
  n108 = LandscapeLayerCoords()
  n109 = ComponentMask(None=MaterialFunctionCall(
    Texture=TextureObject() {samplerType=SAMPLERTYPE_Grayscale, texture=&T_CellNoise}
    LandscapeCoordinates=Multiply(A=@n107, B=@n108)
   ) {materialFunction=&Landscape_TextureSample}.Result) {r=true}
  n110 = ScalarParameter() {defaultValue=2, parameterName=GM_GravelMask_v2_Strength}
  n111 = Clamp(None=Power(Base=Multiply(A=@n109, B=@n110), Exp=ScalarParameter() {defaultValue=4, parameterName=GM-GravelVariationMask_02_Pow}))
  NamedRerouteDeclaration(None=@n111) {name=VariationMask_02, nodeColor=<rgba 1 0 0.6562 1>}
  NamedRerouteDeclaration(None=Multiply(A=@n106, B=NamedRerouteUsage())) {name=GM-VariationMask-1_1, nodeColor=<rgba 1 0 0.8203 1>}
  NamedRerouteDeclaration(None=Multiply(A=@n111, B=NamedRerouteUsage())) {name=GM-VariationMask-2_1, nodeColor=<rgba 0 1 0.8125 1>}
  n112 = Max(A=LandscapeLayerSample() {parameterName=BiomeC}, B=NamedRerouteUsage())
  n113 = LandscapeLayerSample() {parameterName=BiomeB}
  n114 = Max(A=LandscapeLayerSample() {parameterName=Asphalt}, B=LandscapeLayerSample() {parameterName=Mud})
  n115 = Max(A=LandscapeLayerSample() {parameterName=Sand}, B=LandscapeLayerSample() {parameterName=Gravel})
  n116 = LandscapeLayerSample() {parameterName=BiomeA}
  n117 = NamedRerouteUsage()
  n118 = Max(A=Max(A=Max(A=@n114, B=@n115), B=Multiply(A=@n116, B=@n117)), B=Multiply(A=NamedRerouteUsage(), B=NamedRerouteUsage()))
  NamedRerouteDeclaration(None=Multiply(A=Max(A=@n112, B=@n113), B=OneMinus(None=@n118))) {name=PostRVT_SnowMask, nodeColor=<rgba 0.125 0 1 1>}
  NamedRerouteDeclaration(None=@n14.CliffMaskNoContrast) {name=CliffMaskNoContras, nodeColor=<rgba 0.7344 1 0 1>}
  NamedRerouteDeclaration(None=@n28) {name=MobileMaterial, nodeColor=<rgba 1 0 0.9844 1>}
  NamedRerouteDeclaration(None=@n14.SlopeMask) {name=SlopeMask, nodeColor=<rgba 1 1 0 1>}
  NamedRerouteDeclaration(None=OneMinus(None=@n3.BaseLayerMask)) {name=SnowBuildupMasp, nodeColor=<rgba 0.1484 1 0 1>}
  NamedRerouteDeclaration(None=@n87.Tundra) {name=BiomeMask-Tundra, nodeColor=<rgba 1 0.5625 0 1>}
  NamedRerouteDeclaration(None=@n87.GrassLand) {name=BiomeMask-Fall, nodeColor=<rgba 0 1 0 1>}
  NamedRerouteDeclaration(None=Multiply(A=NamedRerouteUsage(), B=@n77)) {name=GM-Fall-Grass, nodeColor=<rgba 0 1 0.5078 1>}
  NamedRerouteDeclaration(None=@n64.Alpha) {name=GM-TundraFoliage, nodeColor=<rgba 0 0.9766 1 1>}
  n119 = LinearInterpolate(
   A=ScalarParameter() {defaultValue=-0.5, parameterName=BiomeA-FFSpawnLa}
   B=ScalarParameter() {defaultValue=1.5, parameterName=BiomeA-FFSpawnLb}
   Alpha=LandscapeLayerSample() {desc=ForestFloor, parameterName=BiomeA}
  )
  NamedRerouteDeclaration(None=MaterialFunctionCall(
    Transition Phase=Saturate(None=@n119)
    Height Texture=NamedRerouteUsage()
    Contrast=ScalarParameter() {defaultValue=0.5, parameterName=BiomeA-FFSpawnCont}
   ) {desc=ForestFloor, materialFunction=&HeightLerp}.Alpha) {name=GM-ForestFloor, nodeColor=<rgba 0.5 0 1 1>}
  NamedRerouteDeclaration(None=@n87.CustomBiomeA) {name=BiomeMask-Jungle, nodeColor=<rgba 1 0 0.8203 1>}
  NamedRerouteDeclaration(None=@n14.OutPut CliffMask2) {name=CliffMask, nodeColor=<rgba 0.5 0 1 1>}
  NamedRerouteDeclaration(None=@n14.LandscapeGrassMeshMask) {name=GM-Cliff, nodeColor=<rgba 0.5 0 1 1>}
  n120 = Multiply(A=LandscapeLayerSample() {desc=Flowers, parameterName=BiomeB}, B=Constant() {r=2})
  n121 = Constant() {r=-0.5}
  NamedRerouteDeclaration(None=Saturate(None=Add(A=@n120, B=@n121))) {name=GM-Flowers, nodeColor=<rgba 0.5 0 1 1>}
  NamedRerouteUsage()
  NamedRerouteUsage()
  NamedRerouteUsage()
  NamedRerouteUsage()
  NamedRerouteUsage()
  NamedRerouteUsage()
  NamedRerouteUsage()
  NamedRerouteUsage()
  NamedRerouteUsage()
  NamedRerouteUsage()
  NamedRerouteUsage()
  NamedRerouteUsage()
  OneMinus(None=NamedRerouteUsage())
  Reroute(None=MaterialFunctionCall() {materialFunction=&MF_LS_GrassSwitch}.Result)
  Reroute(None=MaterialFunctionCall() {materialFunction=&MF_LS_GrassSwitch}.Result)
  Reroute(None=MaterialFunctionCall(In=NamedRerouteUsage()) {materialFunction=&MF_LS_GrassSwitch}.Result)
  Reroute(None=MaterialFunctionCall(In=NamedRerouteUsage()) {materialFunction=&MF_LS_GrassSwitch}.Result)
  Reroute(None=MaterialFunctionCall(In=NamedRerouteUsage()) {materialFunction=&MF_LS_GrassSwitch}.Result)
  Reroute(None=MaterialFunctionCall() {materialFunction=&MF_LS_GrassSwitch}.Result)
  Reroute(None=MaterialFunctionCall() {materialFunction=&MF_LS_GrassSwitch}.Result)
  Reroute(None=MaterialFunctionCall() {materialFunction=&MF_LS_GrassSwitch}.Result)
  RuntimeVirtualTextureOutput!(
   BaseColor=MaterialFunctionCall!(BaseColor=@n16.Base Color) {materialFunction=/Game/Athena/Apollo/Environments/Landscape/Materials/Biome/CommonMaterials/MF_ReduceRTV-CompressionArtifacts}.BaseColor
   Specular=@n16.Specular
   Roughness=@n16.Specular
   Normal=@n16.Specular
   WorldHeight=ComponentMask(None=WorldPosition().XYZ) {b=true}
   Opacity
   Mask
   Displacement
   Mask4
  )
  Saturate(None=Subtract(A=ComponentMask(None=@n7) {r=true}, B=NamedRerouteUsage()))
  TextureObject() {samplerType=SAMPLERTYPE_LinearColor, texture=&HighContrastNoise}
  TextureObject() {samplerType=SAMPLERTYPE_LinearColor, texture=&HighContrastNoise}
  Time()
  VectorParameter() {group=PostRVTSnowMasking, parameterName=PostRVT-SnowMaskColor, defaultValue=<rgba 0.6172 0.6795 0.7305 1>}
 ]
}
